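# Main Niagara effect for the Evil Cowboy glider, with green glows, flames, sparks and smoke, pulsing lantern lights and flickering purple coal.
# Emitters attach to bones and sockets of User.Glider, often using spawn groups to select each attach point.
# A system Timeline drives timed glow, spark and smoke bursts, while waveforms pulse the lights.
# GPU sparks sample the skeleton once per emitter; a disabled unoptimized per-particle version remains.
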NiagaraSystem /BRCosmetics/Gadgets/Assets/VinderTech_GliderChute/Glider_EvilCowboy/FX/NS_Glider_EvilCowboy_Main
dependencies {
 usedDataInterfaces [
  &NiagaraDataInterfaceColorCurve
  &NiagaraDataInterfaceCurve
  &NiagaraDataInterfaceVectorCurve
  &NiagaraDataInterfaceArrayInt32
  &NiagaraDataInterfaceArrayFloat3
  &NiagaraDataInterfaceArrayFloat
  &NiagaraDataInterfaceVector2DCurve
 ]
 usedDynamicInputs [&RemapRange]
 usedModules [
  &SystemState
  &self:SystemSpawnScript.SystemScriptSource.SystemScriptGraph.NiagaraNodeAssignment_0.SetVariables_#1
  &Timeline
  &self:DetermineMovement
 ]
}
emitters {
 Burst_Glow_and_Mesh {
  emitterUpdateScript [
   &EmitterState(Life Cycle Mode=::NewEnumerator1, Loop Behavior=::NewEnumerator1, Loop Duration Mode=::NewEnumerator1, UseLoopDelay=false)
   &SpawnBurst_Instantaneous(Use Spawn Probability=false, UseLoopCountLimit=false)
   &SpawnBurst_Instantaneous(Spawn Group=1, Use Spawn Probability=false, UseLoopCountLimit=false) {module=SpawnBurst_Instantaneous001}
   &SpawnBurst_Instantaneous(Spawn Group=2, Use Spawn Probability=false, UseLoopCountLimit=false) {module=SpawnBurst_Instantaneous002}
   &SampleSkeletalMesh(
    AdditionalYawPitchRoll=false
    Bone / Socket Index=7
    Bone Sampling Mode:ENiagaraBoneSamplingMode=::NewEnumerator4
    Skeletal Mesh:NiagaraDataInterfaceSkeletalMesh=$User.Glider
   ) {module=SampleSkeletalMesh001}
   &SampleSkeletalMesh(
    AdditionalYawPitchRoll=false
    Bone / Socket Index=1
    Mesh Sampling Type:ENiagaraSkelSamplingModeFull=::NewEnumerator5
    Skeletal Mesh:NiagaraDataInterfaceSkeletalMesh=$User.Glider
    Socket Sampling Mode:ENiagaraSocketSamplingMode=::NewEnumerator5
   ) {module=SampleSkeletalMesh002}
   &SampleSkeletalMesh(
    AdditionalYawPitchRoll=false
    Bone / Socket Index=2
    Mesh Sampling Type:ENiagaraSkelSamplingModeFull=::NewEnumerator5
    Skeletal Mesh:NiagaraDataInterfaceSkeletalMesh=$User.Glider
    Socket Sampling Mode:ENiagaraSocketSamplingMode=::NewEnumerator5
   ) {module=SampleSkeletalMesh003}
  ]
  particleSpawnScript [
   &InitializeParticle(
    Color=<rgba 0.6581 1 0.0969 1>
    Lifetime=5
    Mesh Scale=&MaskVectorBySpawnGroup(Masked Vector=<1 2.2 2.2>, Passthrough Vector=<0.75 1.4 1.4>)
    Mesh Scale Mode=::NewEnumerator4
    Sprite Size Mode=::NewEnumerator3
    Uniform Sprite Size=&MaskFloatBySpawnGroup(Masked Float=220, Passthrough Float=50)
    UsePositionOffset=false
    Write Material Random=true
   )
   &InitialMeshOrientation(
    Mesh Orientation Mode=::NewEnumerator2
    Orientation Coordinate Space=::Local
    Orientation Vector=&MaskVectorBySpawnGroup(
     Masked Vector=&RotateVectorByQuaternion(Vector To Rotate=<1 0 0>, Quaternion=$Emitter.SampleSkeletalMesh001.SampledRotation)
     Passthrough Vector=&MaskVectorBySpawnGroup(
      Masked Vector=&RotateVectorByQuaternion(Vector To Rotate=<-1 0 0>, Quaternion=$Emitter.SampleSkeletalMesh002.SampledRotation)
      Passthrough Vector=&RotateVectorByQuaternion(Vector To Rotate=<-1 0 0>, Quaternion=$Emitter.SampleSkeletalMesh003.SampledRotation)
      Spawn Group Mask=1
     )
    )
    Use Rotation Vector=false
   )
  ]
  particleUpdateScript [
   &ParticleState(Kill Particles When Lifetime Has Elapsed=false, Loop Particles Lifetime=true)
   &ScaleColor(
    Scale Alpha=&FloatFromCurve(
     FloatCurve=DataInterface() {propertyValues {Curve {keys [
         {arriveTangent=127.4122, interpMode=RCIM_Cubic, time=0, value=0}
         <at 0.02 1 RCIM_Cubic>
         {interpMode=RCIM_Cubic, leaveTangent=-0.0405, time=0.2, value=0}
        ]}}}
     CurveIndex=$System.Timeline001.NormalizedTime
    )
    Scale RGB=&VectorFromFloat(Value=&FloatFromCurve(
      FloatCurve=DataInterface() {propertyValues {Curve {keys [
          {arriveTangent=689.7969, interpMode=RCIM_Cubic, leaveTangent=689.7969, tangentMode=RCTM_User, time=0, value=1}
          {arriveTangent=-2.5249, interpMode=RCIM_Cubic, leaveTangent=-2.5249, tangentMode=RCTM_User, time=0.02, value=6}
          {arriveTangent=-0.0405, interpMode=RCIM_Cubic, leaveTangent=-0.0405, tangentMode=RCTM_User, time=0.2, value=1}
         ]}}}
      CurveIndex=$System.Timeline001.NormalizedTime
     ))
    ScaleA=true
    ScaleRGB=true
   )
   &CameraOffset(Camera Offset Amount=&FloatFromCurve(
     FloatCurve=DataInterface() {propertyValues {Curve {keys [{arriveTangent=648.8907, interpMode=RCIM_Cubic, leaveTangent=648.8907, tangentMode=RCTM_User, time=0, value=-10}, <at 0.2 50 RCIM_Cubic>]}}}
     CurveIndex=$System.Timeline001.NormalizedTime
    ))
   /SetVariables(Particles.Position=&Add_Vector(
     A=&MaskVectorBySpawnGroup(
      Masked Vector=$Emitter.SampleSkeletalMesh001.SampledPosition
      Passthrough Vector=&MaskVectorBySpawnGroup(
       Masked Vector=$Emitter.SampleSkeletalMesh002.SampledPosition
       Passthrough Vector=$Emitter.SampleSkeletalMesh003.SampledPosition
       Spawn Group Mask=1
      )
     )
     B=&TransformVector(Vector=&MaskVectorBySpawnGroup(Masked Vector=<0 0 15>, Passthrough Vector=<0 0 0>), Destination Space=::World)
    )) {isSetParameters=true}
   /SetVariables(
    Particles.MeshColor:LinearColor=&MakeLinearColorFromVectorAndFloat(
     "Vector (RGB)"=&Multiply_Vector(
      A=&MakeVectorFromLinearColorRGB(Color=<rgba 1.9744 3 0.2906 1>)
      B=&VectorFromFloat(Value=&FloatFromCurve(
        FloatCurve=DataInterface() {propertyValues {Curve {keys [<at 0 1 RCIM_Cubic>,<at 0.02 6 RCIM_Cubic>,<at 0.2 1 RCIM_Cubic>]}}}
        CurveIndex=$System.Timeline001.NormalizedTime
       ))
     )
     "Float (Alpha)"=&FloatFromCurve(
      FloatCurve=DataInterface() {propertyValues {Curve {keys [
          {arriveTangent=127.4122, interpMode=RCIM_Cubic, time=0, value=0}
          <at 0.02 1 RCIM_Cubic>
          {interpMode=RCIM_Cubic, leaveTangent=-0.0405, time=0.2, value=0}
         ]}}}
      CurveIndex=$System.Timeline001.NormalizedTime
     )
    )
    Particles.MeshPosition:NiagaraPosition=&MaskVectorBySpawnGroup(
     Masked Vector=$Emitter.SampleSkeletalMesh001.SampledPosition
     Passthrough Vector=&MaskVectorBySpawnGroup(
      Masked Vector=$Emitter.SampleSkeletalMesh002.SampledPosition
      Passthrough Vector=$Emitter.SampleSkeletalMesh003.SampledPosition
      Spawn Group Mask=1
     )
    )
   ) {isSetParameters=true}
   &ScaleMeshSize(Scale Factor=&VectorFromCurve(
     VectorCurve=DataInterface() {propertyValues {
       XCurve {keys [
         {arriveTangent=17.5185, interpMode=RCIM_Cubic, leaveTangent=17.5185, tangentMode=RCTM_User, time=0, value=0}
         {arriveTangent=10, interpMode=RCIM_Cubic, leaveTangent=10, time=0.0379, value=0.5747}
         <at 0.2 2 RCIM_Cubic>
        ]}
       YCurve {keys [
         {arriveTangent=40.0448, interpMode=RCIM_Cubic, leaveTangent=40.0448, tangentMode=RCTM_User, time=0, value=0}
         <at 0.0379 1 RCIM_Cubic>
         <at 0.2 0 RCIM_Cubic>
        ]}
       ZCurve {keys [
         {arriveTangent=40.5341, interpMode=RCIM_Cubic, leaveTangent=40.5341, tangentMode=RCTM_User, time=0, value=0}
         <at 0.0379 1 RCIM_Cubic>
         <at 0.2 0 RCIM_Cubic>
        ]}
      }}
     CurveIndex=$System.Timeline001.NormalizedTime
    ))
  ]
  rendererClasses [&NiagaraSpriteRendererProperties, &NiagaraMeshRendererProperties]
  renderers [
   {rendererClass=&NiagaraSpriteRendererProperties}
   {rendererClass=&NiagaraMeshRendererProperties, rendererIndex=1}
  ]
 }
 Burst_Sparks {
  simTarget=GPUComputeSim
  emitterUpdateScript [
   &EmitterState()
   &SpawnRate(
    SpawnRate=&Multiply_Float(
     A=100
     B=&FloatFromCurve(
      FloatCurve=DataInterface() {propertyValues {Curve {keys [<at 0 0 RCIM_Constant>,<at 0.01 1 RCIM_Constant>,<at 0.51 0 RCIM_Constant>]}}}
      CurveIndex=$System.Timeline001.NormalizedTime
     )
    )
    Use Spawn Probability=false
   )
   &SpawnRate(
    SpawnGroup=1
    SpawnRate=&Multiply_Float(
     A=150
     B=&FloatFromCurve(
      FloatCurve=DataInterface() {propertyValues {Curve {keys [<at 0 0 RCIM_Constant>,<at 0.01 1 RCIM_Constant>,<at 0.51 0 RCIM_Constant>]}}}
      CurveIndex=$System.Timeline001.NormalizedTime
     )
    )
    Use Spawn Probability=false
   ) {module=SpawnRate001}
   &SpawnRate(
    SpawnGroup=2
    SpawnRate=&Multiply_Float(
     A=150
     B=&FloatFromCurve(
      FloatCurve=DataInterface() {propertyValues {Curve {keys [<at 0 0 RCIM_Constant>,<at 0.01 1 RCIM_Constant>,<at 0.51 0 RCIM_Constant>]}}}
      CurveIndex=$System.Timeline001.NormalizedTime
     )
    )
    Use Spawn Probability=false
   ) {module=SpawnRate002}
   &self:BurstSparks_Sample(FilteredBoneIndex=7, Skeletal Mesh:NiagaraDataInterfaceSkeletalMesh=$User.Glider, SocketA=1, SocketB=2) {module=BurstSparks_Sample}
  ]
  particleSpawnScript [
   &InitializeParticle(
    Color=<rgba 0.6581 1 0.0969 1>
    Lifetime Max=0.4
    Lifetime Min=0.12
    Lifetime Mode=::NewEnumerator1
    Sprite Size Max=<1.2 1.5>
    Sprite Size Min=<0.7 0.5>
    Sprite Size Mode=::NewEnumerator2
    UsePositionOffset=false
    Write Material Random=true
   )
   &ShapeLocation(
    Apply To Particle Position=::NewEnumerator1
    MaskSpawnGroup=false
    Offset=<5 0 0>
    Offset Mode=::NewEnumerator0
    Override Seed=false
    Sphere Radius=&MaskFloatBySpawnGroup(
     Masked Float=&RandomRangeFloat(Minimum=5, Maximum=10)
     Passthrough Float=&RandomRangeFloat(Minimum=2, Maximum=5)
    )
    UseOwnerScale=true
   )
   &self:BurstSparks_ApplySample(
    Offset:Vector3f=&TransformPosition(Position=$Particles.Initial.ShapeLocation.ShapePosition, Source Space=::World, Destination Space=::Local)
    Position0:NiagaraPosition=$Emitter.BurstSparks_Sample.Position0
    Position1:NiagaraPosition=$Emitter.BurstSparks_Sample.Position1
    Position2:NiagaraPosition=$Emitter.BurstSparks_Sample.Position2
    Rotation0:Quat4f=$Emitter.BurstSparks_Sample.Rotation0
    Rotation1:Quat4f=$Emitter.BurstSparks_Sample.Rotation1
    Rotation2:Quat4f=$Emitter.BurstSparks_Sample.Rotation2
   ) {module=BurstSparks_ApplySample}
   &AddVelocity(
    Cone Angle=8
    Cone Axis=&MaskVectorBySpawnGroup(Masked Vector=<0 0 1>, Passthrough Vector=<0 1 0>)
    Module.UseConeFalloff=false
    Module.UseVelDistribution=false
    Override Seed=false
    Velocity Mode=::NewEnumerator2
    Velocity Speed=&RandomRangeFloat(Minimum=250, Maximum=750)
   )
  ]
  particleUpdateScript [
   &ParticleState()
   &ScaleColor(
    Scale Alpha=&FloatFromCurve(FloatCurve=DataInterface() {propertyValues {Curve {keys [{arriveTangent=-14.7885, interpMode=RCIM_Cubic, leaveTangent=-14.7885, tangentMode=RCTM_User, time=0, value=2}, <at 0.2 1 RCIM_Cubic>]}}})
    Scale RGB=&VectorFromFloat(Value=&FloatFromCurve(FloatCurve=DataInterface() {propertyValues {Curve {keys [<at 0 16 RCIM_Cubic>,<at 0.8 1 RCIM_Cubic>]}}}))
    ScaleA=true
    ScaleRGB=true
   )
   &AccelerationForce(Acceleration=<0 0 -1200>)
   &CurlNoiseForce(Noise Frequency=150, Noise Quality / Cost=::NewEnumerator0, Noise Strength=1500, Pan Noise=true, Pan Noise Field=<0 25 -25>)
   &Drag(Drag=&RandomRangeFloat(Minimum=0, Maximum=2), Use Linear Drag=true, Use Rotational Drag=false)
   &SolveForcesAndVelocity(Clamp Velocity=false, Limit Acceleration=false)
   &ScaleSpriteSize(Uniform Curve Sprite Scale=DataInterface() {propertyValues {Curve {keys [{arriveTangent=99.8418, interpMode=RCIM_Cubic, leaveTangent=99.8418, tangentMode=RCTM_User, time=0, value=0}, <at 0.03 1 RCIM_Cubic>]}}})
   &ScaleSpriteSizeBySpeed(Max Scale Factor=<1 10>, Min Scale Factor=<0 1>)
  ]
  rendererClasses [&NiagaraSpriteRendererProperties]
  renderers [{rendererClass=&NiagaraSpriteRendererProperties}]
 }
 Burst_Sparks_Unopt {
  bEnabled=false
  simTarget=GPUComputeSim
  emitterUpdateScript [
   &EmitterState()
   &SpawnRate(
    SpawnRate=&Multiply_Float(
     A=100
     B=&FloatFromCurve(
      FloatCurve=DataInterface() {propertyValues {Curve {keys [<at 0 0 RCIM_Constant>,<at 0.01 1 RCIM_Constant>,<at 0.51 0 RCIM_Constant>]}}}
      CurveIndex=$System.Timeline001.NormalizedTime
     )
    )
    Use Spawn Probability=false
   )
   &SpawnRate(
    SpawnGroup=1
    SpawnRate=&Multiply_Float(
     A=150
     B=&FloatFromCurve(
      FloatCurve=DataInterface() {propertyValues {Curve {keys [<at 0 0 RCIM_Constant>,<at 0.01 1 RCIM_Constant>,<at 0.51 0 RCIM_Constant>]}}}
      CurveIndex=$System.Timeline001.NormalizedTime
     )
    )
    Use Spawn Probability=false
   ) {module=SpawnRate001}
   &SpawnRate(
    SpawnGroup=2
    SpawnRate=&Multiply_Float(
     A=150
     B=&FloatFromCurve(
      FloatCurve=DataInterface() {propertyValues {Curve {keys [<at 0 0 RCIM_Constant>,<at 0.01 1 RCIM_Constant>,<at 0.51 0 RCIM_Constant>]}}}
      CurveIndex=$System.Timeline001.NormalizedTime
     )
    )
    Use Spawn Probability=false
   ) {module=SpawnRate002}
  ]
  particleSpawnScript [
   &InitializeParticle(
    Color=<rgba 0.6581 1 0.0969 1>
    Lifetime Max=0.4
    Lifetime Min=0.12
    Lifetime Mode=::NewEnumerator1
    Sprite Size Max=<1.2 1.5>
    Sprite Size Min=<0.7 0.5>
    Sprite Size Mode=::NewEnumerator2
    UsePositionOffset=false
    Write Material Random=true
   )
   &ShapeLocation(
    Apply To Particle Position=::NewEnumerator1
    MaskSpawnGroup=false
    Offset=<5 0 0>
    Offset Mode=::NewEnumerator0
    Override Seed=false
    Sphere Radius=&MaskFloatBySpawnGroup(
     Masked Float=&RandomRangeFloat(Minimum=5, Maximum=10)
     Passthrough Float=&RandomRangeFloat(Minimum=2, Maximum=5)
    )
    UseOwnerScale=true
   )
   &AddVelocity(
    Cone Angle=8
    Cone Axis=&MaskVectorBySpawnGroup(Masked Vector=<0 0 1>, Passthrough Vector=<0 1 0>)
    Module.UseConeFalloff=false
    Module.UseVelDistribution=false
    Override Seed=false
    Velocity Mode=::NewEnumerator2
    Velocity Speed=&RandomRangeFloat(Minimum=250, Maximum=750)
   )
  ]
  particleUpdateScript [
   &ParticleState()
   &ScaleColor(
    Scale Alpha=&FloatFromCurve(FloatCurve=DataInterface() {propertyValues {Curve {keys [{arriveTangent=-14.7885, interpMode=RCIM_Cubic, leaveTangent=-14.7885, tangentMode=RCTM_User, time=0, value=2}, <at 0.2 1 RCIM_Cubic>]}}})
    Scale RGB=&VectorFromFloat(Value=&FloatFromCurve(FloatCurve=DataInterface() {propertyValues {Curve {keys [<at 0 16 RCIM_Cubic>,<at 0.8 1 RCIM_Cubic>]}}}))
    ScaleA=true
    ScaleRGB=true
   )
   &SkeletalMeshLocation(
    Bone / Socket Index=7
    Bone Sampling Mode=::NewEnumerator4
    MaskSpawnGroup=true
    Position Sampling=::NewEnumerator1
    Sampled Position Offset=/Niagara/DynamicInputs/Transforms/TransformPosition(Position=$Particles.ShapeLocation.ShapePosition, Source Space=::World, Destination Space=::Local)
    Skeletal Mesh=$User.Glider
   )
   &SkeletalMeshLocation(
    Bone / Socket Index=1
    MaskSpawnGroup=true
    Mesh Sampling Type=::NewEnumerator5
    Sampled Position Offset=&TransformPosition(Position=$Particles.ShapeLocation.ShapePosition, Source Space=::World, Destination Space=::Local)
    Skeletal Mesh=$User.Glider
    Socket Sampling Mode=::NewEnumerator5
    Spawn Group Mask=1
   ) {module=SkeletalMeshLocation001}
   &SkeletalMeshLocation(
    Bone / Socket Index=2
    MaskSpawnGroup=true
    Mesh Sampling Type=::NewEnumerator5
    Sampled Position Offset=&TransformPosition(Position=$Particles.ShapeLocation.ShapePosition, Source Space=::World, Destination Space=::Local)
    Skeletal Mesh=$User.Glider
    Socket Sampling Mode=::NewEnumerator5
    Spawn Group Mask=2
   ) {module=SkeletalMeshLocation002}
   &AccelerationForce(Acceleration=<0 0 -1200>)
   &CurlNoiseForce(Noise Frequency=150, Noise Quality / Cost=::NewEnumerator0, Noise Strength=1500, Pan Noise=true, Pan Noise Field=<0 25 -25>)
   &Drag(Drag=&RandomRangeFloat(Minimum=0, Maximum=2), Use Linear Drag=true, Use Rotational Drag=false)
   &SolveForcesAndVelocity(Clamp Velocity=false, Limit Acceleration=false)
   &ScaleSpriteSize(Uniform Curve Sprite Scale=DataInterface() {propertyValues {Curve {keys [{arriveTangent=99.8418, interpMode=RCIM_Cubic, leaveTangent=99.8418, tangentMode=RCTM_User, time=0, value=0}, <at 0.03 1 RCIM_Cubic>]}}})
   &ScaleSpriteSizeBySpeed(Max Scale Factor=<1 10>, Min Scale Factor=<0 1>)
  ]
  rendererClasses [&NiagaraSpriteRendererProperties]
  renderers [{rendererClass=&NiagaraSpriteRendererProperties}]
 }
 Coal {
  emitterUpdateScript [
   &EmitterState(
    Life Cycle Mode=::NewEnumerator1
    Loop Delay=&RandomRangeFloat(Minimum=0.2, Maximum=2)
    Loop Duration=0.2
    UseLoopDelay=true
   )
   &SpawnBurst_Instantaneous(Spawn Count=10, Use Spawn Probability=false, UseLoopCountLimit=false)
   &SpawnBurst_Instantaneous(Spawn Group=1, Use Spawn Probability=false, UseLoopCountLimit=false) {module=SpawnBurst_Instantaneous001}
   &SpawnBurst_Instantaneous(Spawn Group=2, Spawn Time=0.05, Use Spawn Probability=false, UseLoopCountLimit=false) {module=SpawnBurst_Instantaneous002}
   &SpawnBurst_Instantaneous(Spawn Group=3, Spawn Time=0.1, Use Spawn Probability=false, UseLoopCountLimit=false) {module=SpawnBurst_Instantaneous003}
  ]
  particleSpawnScript [
   /SetVariables(Particles.VisibilityTag=&SelectIntFromArray(Direct Array Index=&SpawnGroup(), Int Selection Array=DataInterface() {propertyValues {IntData [0,1,1,1]}})) {isSetParameters=true}
   &InitializeParticle(
    Color=<rgba 2.7749 0.1518 5 0.5>
    Lifetime Max=0.25
    Lifetime Min=0.15
    Lifetime Mode=::NewEnumerator1
    Mesh Scale Max=<0.7 0.7 0.7>
    Mesh Scale Min=<0.7 0.5 0.5>
    Mesh Scale Mode=::NewEnumerator2
    Sprite Rotation Mode=::NewEnumerator2
    Sprite Size Max=<10 15>
    Sprite Size Mode=::NewEnumerator2
    UsePositionOffset=false
    Write Material Random=true
   )
   &CameraOffset(Camera Offset Amount=10)
   &AddVelocity(Velocity=&MaskVectorBySpawnGroup(
     Masked Vector=&RandomRangeVector(Minimum=<0.5 0 2>, Maximum=<1 0 5>)
     Passthrough Vector=&VectorFromCurve(VectorCurve=DataInterface() {propertyValues {
        XCurve {keys [<at 0 0 RCIM_Constant>,<at 0.025 -5 RCIM_Constant>,<at 0.075 0 RCIM_Constant>]}
        YCurve {keys [<at 0 0 RCIM_Constant>,<at 0.025 0 RCIM_Constant>,<at 0.075 -5 RCIM_Constant>]}
        ZCurve {keys [<at 0 5 RCIM_Constant>,<at 0.025 0 RCIM_Constant>,<at 0.075 0 RCIM_Constant>]}
       }})
    ))
  ]
  particleUpdateScript [
   &ParticleState()
   &SolveForcesAndVelocity(Clamp Velocity=false, Limit Acceleration=false)
   &SkeletalMeshLocation(
    Bone Sampling Mode=::NewEnumerator4
    MaskSpawnGroup=false
    Sampled Position Offset=&MaskVectorBySpawnGroup(
     Masked Vector=&RandomRangeVector(Minimum=<-25.4783 -30 0>, Maximum=<0 30 30>)
     Passthrough Vector=&SelectVectorFromArray(
      Direct Array Index=&SpawnGroup()
      Vector Selection Array=DataInterface() {propertyValues {FloatData [<0 0 0>, <-18.8352 0 25>, <-17.1536 20.926 18.2654>, <-12.7978 -19.7667 19.2867>]}}
     )
    )
    Skeletal Mesh=$User.Glider
   )
   &DynamicMaterialParameters(
    Index 0 Param 1=&RandomRangeFloat(Minimum=2.5, Maximum=3.5)
    Index 0 Param 2=&FloatFromCurve(FloatCurve=DataInterface() {propertyValues {Curve {keys [<at 0 1>,<at 1 -1>]}}})
    Index 0 Param 3=&RandomRangeFloat(Minimum=0.2, Maximum=0.7)
    Module.Param0WriteEnabled=true
    Module.Param1WriteEnabled=true
    Module.Param2WriteEnabled=true
    Module.Param3WriteEnabled=true
   )
   /SetVariables(Particles.Mesh Dynamic Parameter:Vector4f=&MakeVector4(
     X=&RandomRangeFloat(Minimum=2, Maximum=3)
     Y=&FloatFromCurve(FloatCurve=DataInterface() {propertyValues {Curve {keys [{arriveTangent=-2.0678, interpMode=RCIM_Cubic, leaveTangent=-2.0678, tangentMode=RCTM_User, time=0, value=1}, <at 1 -0.25 RCIM_Cubic>]}}})
     Z=&RandomRangeFloat(Minimum=0.2, Maximum=2)
     W=&RandomRangeFloat(Minimum=0, Maximum=1)
    )) {isSetParameters=true}
   &ScaleColor(
    Scale RGB=&MakeVectorFromLinearColorRGB(Color=&ColorFromCurve(ColorCurve=DataInterface() {propertyValues {
        AlphaCurve {keys []}
        BlueCurve {keys [<at 0 10 RCIM_Cubic>,<at 0.5 5>,<at 1 2 RCIM_Cubic>]}
        GreenCurve {keys [<at 0 0.3035 RCIM_Cubic>,<at 0.5 0.1518>,<at 1 0.0607 RCIM_Cubic>]}
        RedCurve {keys [<at 0 5.5498 RCIM_Cubic>,<at 0.5 2.7749>,<at 1 1.11 RCIM_Cubic>]}
       }}))
    ScaleA=true
    ScaleRGB=true
   )
   &ScaleMeshSize(Scale Factor=&VectorFromCurve(VectorCurve=DataInterface() {propertyValues {
       XCurve {keys [<at 0 0.2 RCIM_Constant>,<at 0.2 0.6 RCIM_Constant>,<at 0.4 0.7606 RCIM_Constant>,<at 0.6 0.8999 RCIM_Constant>,<at 0.8 1 RCIM_Constant>]}
       YCurve {keys [<at 0 1 RCIM_Constant>,<at 0.2 0.8291 RCIM_Constant>,<at 0.4 0.4373 RCIM_Constant>,<at 0.6 0.3041 RCIM_Constant>,<at 0.8 0.2 RCIM_Constant>]}
       ZCurve {keys [<at 0 1 RCIM_Constant>,<at 0.2 0.691 RCIM_Constant>,<at 0.4 0.4457 RCIM_Constant>,<at 0.6 0.5668 RCIM_Constant>,<at 0.8 0.6282 RCIM_Constant>]}
      }}))
   /SetVariables() {isSetParameters=true}
  ]
  rendererClasses [&NiagaraSpriteRendererProperties, &NiagaraMeshRendererProperties]
  renderers [
   {rendererClass=&NiagaraSpriteRendererProperties}
   {rendererClass=&NiagaraMeshRendererProperties, rendererIndex=1}
  ]
 }
 Coal_Light {
  emitterUpdateScript [
   &EmitterState(Life Cycle Mode=::NewEnumerator1, Loop Behavior=::NewEnumerator1, Loop Duration Mode=::NewEnumerator1, UseLoopDelay=false)
   &SpawnBurst_Instantaneous(Use Spawn Probability=false, UseLoopCountLimit=false)
  ]
  particleSpawnScript [&InitializeParticle(Color=<rgba 2.7749 0.1518 5 1>, Lifetime=0, UsePositionOffset=false, Write Material Random=true)]
  particleUpdateScript [
   &ParticleState(Kill Particles When Lifetime Has Elapsed=false, Let Infinitely Lived Particles Die When Emitter Deactivates=true)
   &SkeletalMeshLocation(Bone Sampling Mode=::NewEnumerator4, MaskSpawnGroup=false, Sampled Position Offset=<-30 0 15>, Skeletal Mesh=$User.Glider)
   /SetVariables(Particles.LightExponent=2, Particles.LightRadius=50) {isSetParameters=true}
   &ScaleColor(
    Scale RGB=&VectorFromFloat(Value=&Multiply_Float(
      A=150
      B=&Waveform("Waveform [1]"=::NewEnumerator9, "[1] Frequency"=1, "[1] Amplitude Min/Max"=<1 2>, "[1] Pulse Duration"=0.5)
     ))
    ScaleA=true
    ScaleRGB=true
   )
  ]
  rendererClasses [&NiagaraLightRendererProperties]
  renderers [{rendererClass=&NiagaraLightRendererProperties, rendererIndex=0}]
 }
 Glow_and_Lights {
  emitterUpdateScript [
   &EmitterState(Life Cycle Mode=::NewEnumerator1, Loop Behavior=::NewEnumerator1, Loop Duration Mode=::NewEnumerator1, UseLoopDelay=false)
   &SpawnBurst_Instantaneous(Use Spawn Probability=false, UseLoopCountLimit=false)
   &SpawnBurst_Instantaneous(Spawn Group=1, Use Spawn Probability=false, UseLoopCountLimit=false) {module=SpawnBurst_Instantaneous001}
   &SpawnBurst_Instantaneous(Spawn Group=2, Use Spawn Probability=false, UseLoopCountLimit=false) {module=SpawnBurst_Instantaneous002}
   /SetVariables(Emitter.Controller:NiagaraFloat=&Waveform("Waveform [1]"=::NewEnumerator9, "[1] Frequency"=0.125, "[1] Amplitude Min/Max"=<0 1>)) {isSetParameters=true}
  ]
  particleSpawnScript [
   &InitializeParticle(
    Color=<rgba 0.4339 1 0 0.1>
    Lifetime=0
    Mesh Scale Mode=::NewEnumerator3
    Mesh Uniform Scale=0.75
    UsePositionOffset=false
    Write Material Random=true
   )
   &CameraOffset(Camera Offset Amount=&MaskFloatBySpawnGroup(Masked Float=40, Passthrough Float=20))
  ]
  particleUpdateScript [
   &ParticleState(Kill Particles When Lifetime Has Elapsed=false, Let Infinitely Lived Particles Die When Emitter Deactivates=true)
   &SkeletalMeshLocation(
    Bone / Socket Index=7
    Bone Sampling Mode=::NewEnumerator4
    MaskSpawnGroup=true
    Orientation Sampling=::NewEnumerator1
    Position Sampling=::NewEnumerator2
    Sampled Position Offset=<15 0 0>
    Skeletal Mesh=$User.Glider
   )
   &SkeletalMeshLocation(
    Bone / Socket Index=&SpawnGroup()
    MaskSpawnGroup=false
    Mesh Sampling Type=::NewEnumerator5
    Orientation Sampling=::NewEnumerator1
    Position Sampling=::NewEnumerator2
    Sampled Position Offset=&MaskVectorBySpawnGroup(Masked Vector=<0.5 3 -1.5>, Passthrough Vector=<-0.5 3 1.5>, Spawn Group Mask=1)
    Skeletal Mesh=$User.Glider
    Socket Sampling Mode=::NewEnumerator5
   ) {module=SkeletalMeshLocation001}
   /SetVariables(Particles.Position=&MaskVectorBySpawnGroup(
     Masked Vector=$Particles.SkeletalMeshLocation.SampledPosition
     Passthrough Vector=Linked() {linkedVariable {name=Particles.SkeletalMeshLocation001.SampledPosition, type=NiagaraPosition}}
    )) {isSetParameters=true}
   &ScaleColor(
    Scale RGB=&VectorFromCurve(
     VectorCurve=DataInterface() {propertyValues {
       XCurve {keys [
         {arriveTangent=3.8323, interpMode=RCIM_Cubic, leaveTangent=3.8323, leaveTangentWeight=0.1667, tangentMode=RCTM_User, time=0, value=0}
         {arriveTangentWeight=0.1667, interpMode=RCIM_Cubic, leaveTangentWeight=0.1667, tangentMode=RCTM_User, time=0.5, value=1}
         {arriveTangent=-4.5086, arriveTangentWeight=0.1667, interpMode=RCIM_Cubic, leaveTangent=-4.5086, tangentMode=RCTM_User, time=1, value=0}
        ]}
       YCurve {keys [
         {arriveTangent=2.7227, interpMode=RCIM_Cubic, leaveTangent=2.7227, leaveTangentWeight=0.1667, tangentMode=RCTM_User, time=0, value=0}
         {arriveTangentWeight=0.1667, interpMode=RCIM_Cubic, leaveTangentWeight=0.1667, tangentMode=RCTM_User, time=0.5, value=0.9346}
         {arriveTangent=-2.1767, arriveTangentWeight=0.1667, interpMode=RCIM_Cubic, leaveTangent=-2.1767, tangentMode=RCTM_User, time=1, value=0}
        ]}
       ZCurve {keys [
         {arriveTangent=0.2747, interpMode=RCIM_Cubic, leaveTangent=0.2747, leaveTangentWeight=0.1667, tangentMode=RCTM_User, time=0, value=0}
         {arriveTangentWeight=0.1667, interpMode=RCIM_Cubic, leaveTangentWeight=0.1667, tangentMode=RCTM_User, time=0.5, value=0.1159}
         {arriveTangent=-0.3218, arriveTangentWeight=0.1667, interpMode=RCIM_Cubic, leaveTangent=-0.3218, tangentMode=RCTM_User, time=1, value=0}
        ]}
      }}
     CurveIndex=$Emitter.Controller
    )
    ScaleA=true
    ScaleRGB=true
   )
   &DynamicMaterialParameters(
    Index 0 Param 1=&MaskFloatBySpawnGroup(Masked Float=1, Passthrough Float=0.75)
    Index 0 Param 2=15
    Index 0 Param 3=&MaskFloatBySpawnGroup(Masked Float=15, Passthrough Float=5)
    Index 0 Param 4=15
    Module.Param0WriteEnabled=true
    Module.Param1WriteEnabled=true
    Module.Param2WriteEnabled=true
    Module.Param3WriteEnabled=true
   )
   /SetVariables(
    Particles.LightColor=&MakeLinearColorFromVectorAndFloat(
     "Vector (RGB)"=&Multiply_Vector(
      A=&MakeVectorFromLinearColorRGB(Color=&ColorFromCurve(
        ColorCurve=DataInterface() {propertyValues {
          AlphaCurve {keys []}
          BlueCurve {keys [
            {arriveTangent=0.2259, interpMode=RCIM_Cubic, leaveTangent=0.2259, tangentMode=RCTM_User, time=0, value=0}
            <at 0.5 0.0781 RCIM_Cubic>
            {arriveTangent=-0.2204, interpMode=RCIM_Cubic, leaveTangent=-0.2204, tangentMode=RCTM_User, time=1, value=0}
           ]}
          GreenCurve {keys [
            {arriveTangent=6.2787, interpMode=RCIM_Cubic, leaveTangent=6.2787, tangentMode=RCTM_User, time=0, value=1}
            <at 0.5 2.5 RCIM_Cubic>
            {arriveTangent=-6.4626, interpMode=RCIM_Cubic, leaveTangent=-6.4626, tangentMode=RCTM_User, time=1, value=1}
           ]}
          RedCurve {keys [
            {arriveTangent=7.1812, interpMode=RCIM_Cubic, leaveTangent=7.1812, tangentMode=RCTM_User, time=0, value=0}
            <at 0.5 1.6927 RCIM_Cubic>
            {arriveTangent=-7.7098, interpMode=RCIM_Cubic, leaveTangent=-7.7098, tangentMode=RCTM_User, time=1, value=0}
           ]}
         }}
        CurveIndex=$Emitter.Controller
       ))
      B=&VectorFromFloat(Value=100)
     )
     "Float (Alpha)"=1
    )
    Particles.LightExponent=2
    Particles.LightRadius=&MaskFloatBySpawnGroup(Masked Float=40, Passthrough Float=25)
   ) {isSetParameters=true}
  ]
  rendererClasses [&NiagaraMeshRendererProperties, &NiagaraLightRendererProperties]
  renderers [
   {rendererClass=&NiagaraMeshRendererProperties, rendererIndex=0}
   {rendererClass=&NiagaraLightRendererProperties, rendererIndex=1}
  ]
 }
 Lantern {
  emitterUpdateScript [
   &EmitterState(Life Cycle Mode=::NewEnumerator1, Loop Behavior=::NewEnumerator1, Loop Duration Mode=::NewEnumerator1, UseLoopDelay=false)
   &SpawnBurst_Instantaneous(Use Spawn Probability=false, UseLoopCountLimit=false)
   /SetVariables(Emitter.LanternControl:NiagaraFloat=&Waveform(Global Phase=$Engine.Time, "Waveform [1]"=::NewEnumerator9, "[1] Frequency"=0.125, "[1] Amplitude Min/Max"=<0 1>)) {isSetParameters=true}
  ]
  particleSpawnScript [
   &InitializeParticle(
    Lifetime=0
    Mesh Scale Mode=::NewEnumerator3
    Mesh Uniform Scale=0.75
    Sprite Size Mode=::NewEnumerator3
    Uniform Sprite Size=100
    UsePositionOffset=false
    Write Material Random=true
   )
   &CameraOffset(Camera Offset Amount=10)
  ]
  particleUpdateScript [
   &ParticleState(Kill Particles When Lifetime Has Elapsed=false, Let Infinitely Lived Particles Die When Emitter Deactivates=true)
   &SkeletalMeshLocation(MaskSpawnGroup=false, Mesh Sampling Type=::NewEnumerator5, Skeletal Mesh=$User.Glider, Socket Sampling Mode=::NewEnumerator5)
   &ScaleColor(
    Scale RGB=&MakeVectorFromLinearColorRGB(Color=&ColorFromCurve(
      ColorCurve=DataInterface() {propertyValues {
        AlphaCurve {keys []}
        BlueCurve {keys [
          {arriveTangent=0.2259, interpMode=RCIM_Cubic, leaveTangent=0.2259, tangentMode=RCTM_User, time=0, value=0}
          <at 0.5 0.0781 RCIM_Cubic>
          {arriveTangent=-0.2204, interpMode=RCIM_Cubic, leaveTangent=-0.2204, tangentMode=RCTM_User, time=1, value=0}
         ]}
        GreenCurve {keys [
          {arriveTangent=6.2787, interpMode=RCIM_Cubic, leaveTangent=6.2787, tangentMode=RCTM_User, time=0, value=1}
          <at 0.5 2.5 RCIM_Cubic>
          {arriveTangent=-6.4626, interpMode=RCIM_Cubic, leaveTangent=-6.4626, tangentMode=RCTM_User, time=1, value=1}
         ]}
        RedCurve {keys [
          {arriveTangent=7.1812, interpMode=RCIM_Cubic, leaveTangent=7.1812, tangentMode=RCTM_User, time=0, value=0}
          <at 0.5 1.6927 RCIM_Cubic>
          {arriveTangent=-7.7098, interpMode=RCIM_Cubic, leaveTangent=-7.7098, tangentMode=RCTM_User, time=1, value=0}
         ]}
       }}
      CurveIndex=$Emitter.LanternControl
     ))
    ScaleA=true
    ScaleRGB=true
   )
   &ScaleSpriteSize(
    Uniform Curve Index=$Emitter.LanternControl
    Uniform Curve Sprite Scale=DataInterface() {propertyValues {Curve {keys [
        {interpMode=RCIM_Cubic, leaveTangentWeight=0.5354, tangentMode=RCTM_User, time=0, value=0.85}
        {arriveTangentWeight=0.1667, interpMode=RCIM_Cubic, leaveTangentWeight=0.1667, tangentMode=RCTM_User, time=0.5, value=1.25}
        {arriveTangentWeight=0.4994, interpMode=RCIM_Cubic, tangentMode=RCTM_User, time=1, value=0.85}
       ]}}}
   )
   /SetVariables(Particles.SpriteRotation:NiagaraFloat=&Lerp_Float(A=0, B=90, Alpha=$Emitter.LanternControl)) {isSetParameters=true}
   /SetVariables(
    Particles.Mesh Camera Offset=40
    Particles.Mesh Color:LinearColor=&MakeLinearColorFromVectorAndFloat(
     "Vector (RGB)"=&Multiply_Vector(
      A=&MakeVectorFromLinearColorRGB(Color=<rgba 0 1 0 1>)
      B=&VectorFromCurve(
       VectorCurve=DataInterface() {propertyValues {
         XCurve {keys [
           {arriveTangent=3.8323, interpMode=RCIM_Cubic, leaveTangent=3.8323, leaveTangentWeight=0.1667, tangentMode=RCTM_User, time=0, value=0}
           {arriveTangentWeight=0.1667, interpMode=RCIM_Cubic, leaveTangentWeight=0.1667, tangentMode=RCTM_User, time=0.5, value=1}
           {arriveTangent=-4.5086, arriveTangentWeight=0.1667, interpMode=RCIM_Cubic, leaveTangent=-4.5086, tangentMode=RCTM_User, time=1, value=0}
          ]}
         YCurve {keys [
           {arriveTangent=2.7227, interpMode=RCIM_Cubic, leaveTangent=2.7227, leaveTangentWeight=0.1667, tangentMode=RCTM_User, time=0, value=0}
           {arriveTangentWeight=0.1667, interpMode=RCIM_Cubic, leaveTangentWeight=0.1667, tangentMode=RCTM_User, time=0.5, value=0.9346}
           {arriveTangent=-2.1767, arriveTangentWeight=0.1667, interpMode=RCIM_Cubic, leaveTangent=-2.1767, tangentMode=RCTM_User, time=1, value=0}
          ]}
         ZCurve {keys [
           {arriveTangent=0.2747, interpMode=RCIM_Cubic, leaveTangent=0.2747, leaveTangentWeight=0.1667, tangentMode=RCTM_User, time=0, value=0}
           {arriveTangentWeight=0.1667, interpMode=RCIM_Cubic, leaveTangentWeight=0.1667, tangentMode=RCTM_User, time=0.5, value=0.1159}
           {arriveTangent=-0.3218, arriveTangentWeight=0.1667, interpMode=RCIM_Cubic, leaveTangent=-0.3218, tangentMode=RCTM_User, time=1, value=0}
          ]}
        }}
       CurveIndex=$Emitter.LanternControl
      )
     )
     "Float (Alpha)"=0.05
    )
    Particles.Mesh Dynamic Parameters:Vector4f=&MakeVector4(
     X=&FloatFromCurve(FloatCurve=DataInterface() {propertyValues {Curve {keys [<at 0 0.25>,<at 1 0.33>]}}}, CurveIndex=$Emitter.LanternControl)
     Y=15
     Z=15
     W=2
    )
   ) {isSetParameters=true}
   /SetVariables(
    Particles.LightColor=&MakeLinearColorFromVectorAndFloat(
     "Vector (RGB)"=&Multiply_Vector(
      A=&MakeVectorFromLinearColorRGB(Color=&ColorFromCurve(
        ColorCurve=DataInterface() {propertyValues {
          AlphaCurve {keys []}
          BlueCurve {keys [
            {arriveTangent=0.2259, interpMode=RCIM_Cubic, leaveTangent=0.2259, tangentMode=RCTM_User, time=0, value=0}
            <at 0.5 0.0781 RCIM_Cubic>
            {arriveTangent=-0.2204, interpMode=RCIM_Cubic, leaveTangent=-0.2204, tangentMode=RCTM_User, time=1, value=0}
           ]}
          GreenCurve {keys [
            {arriveTangent=6.2787, interpMode=RCIM_Cubic, leaveTangent=6.2787, tangentMode=RCTM_User, time=0, value=1}
            <at 0.5 2.5 RCIM_Cubic>
            {arriveTangent=-6.4626, interpMode=RCIM_Cubic, leaveTangent=-6.4626, tangentMode=RCTM_User, time=1, value=1}
           ]}
          RedCurve {keys [
            {arriveTangent=7.1812, interpMode=RCIM_Cubic, leaveTangent=7.1812, tangentMode=RCTM_User, time=0, value=0}
            <at 0.5 1.6927 RCIM_Cubic>
            {arriveTangent=-7.7098, interpMode=RCIM_Cubic, leaveTangent=-7.7098, tangentMode=RCTM_User, time=1, value=0}
           ]}
         }}
        CurveIndex=$Emitter.LanternControl
       ))
      B=&VectorFromFloat(Value=5)
     )
     "Float (Alpha)"=1
    )
    Particles.LightExponent=2
    Particles.LightRadius=40
   ) {isSetParameters=true}
  ]
  rendererClasses [
   &NiagaraSpriteRendererProperties
   &NiagaraMeshRendererProperties
   &NiagaraLightRendererProperties
  ]
  renderers [
   {rendererClass=&NiagaraSpriteRendererProperties}
   {rendererClass=&NiagaraMeshRendererProperties, rendererIndex=1}
   {rendererClass=&NiagaraLightRendererProperties, rendererIndex=2}
  ]
 }
 Smoke {
  emitterUpdateScript [
   &EmitterState()
   &SpawnRate(SpawnRate=10, Use Spawn Probability=false)
   &SpawnRate(
    SpawnRate=&Multiply_Float(
     A=250
     B=&FloatFromCurve(
      FloatCurve=DataInterface() {propertyValues {Curve {keys [<at 0 0 RCIM_Constant>,<at 0.01 1 RCIM_Constant>,<at 0.15 0 RCIM_Constant>]}}}
      CurveIndex=$System.Timeline001.NormalizedTime
     )
    )
    Use Spawn Probability=false
   ) {module=SpawnRate001}
   &SpawnRate(
    SpawnGroup=1
    SpawnRate=&Multiply_Float(
     A=40
     B=&FloatFromCurve(
      FloatCurve=DataInterface() {propertyValues {Curve {keys [<at 0 0 RCIM_Constant>,<at 0.01 1 RCIM_Constant>,<at 0.15 0 RCIM_Constant>]}}}
      CurveIndex=$System.Timeline001.NormalizedTime
     )
    )
    Use Spawn Probability=false
   ) {module=SpawnRate002}
   &SpawnRate(
    SpawnGroup=2
    SpawnRate=&Multiply_Float(
     A=40
     B=&FloatFromCurve(
      FloatCurve=DataInterface() {propertyValues {Curve {keys [<at 0 0 RCIM_Constant>,<at 0.01 1 RCIM_Constant>,<at 0.15 0 RCIM_Constant>]}}}
      CurveIndex=$System.Timeline001.NormalizedTime
     )
    )
    Use Spawn Probability=false
   ) {module=SpawnRate003}
  ]
  particleSpawnScript [
   &InitializeParticle(
    Color=&MakeLinearColorFromVectorAndFloat("Vector (RGB)"=<1 1 1>, "Float (Alpha)"=&RandomRangeFloat(Minimum=0.1, Maximum=0.3))
    Lifetime Max=1
    Lifetime Min=0.5
    Lifetime Mode=::NewEnumerator1
    Sprite Rotation Mode=::NewEnumerator2
    Sprite Size Mode=::NewEnumerator1
    Uniform Sprite Size Max=150
    Uniform Sprite Size Min=80
    UsePositionOffset=false
    Write Material Random=true
   )
   &AddVelocity(Velocity=&RandomRangeVector(Minimum=<0 150 0>, Maximum=<0 300 0>))
  ]
  particleUpdateScript [
   &ParticleState()
   &SkeletalMeshLocation(
    Bone / Socket Index=&SpawnGroup()
    MaskSpawnGroup=false
    Mesh Sampling Type=::NewEnumerator5
    Orientation Sampling=::NewEnumerator1
    Position Sampling=::NewEnumerator1
    Skeletal Mesh=$User.Glider
    Socket Sampling Mode=::NewEnumerator5
    Source Movement Scale:Vector3f=<0 0 0>
   ) {module=SkeletalMeshLocation001}
   &SkeletalMeshLocation(
    Bone / Socket Index=7
    Bone Sampling Mode=::NewEnumerator4
    MaskSpawnGroup=true
    Position Sampling=::NewEnumerator1
    Sampled Position Offset=<10 -5 0>
    Skeletal Mesh=$User.Glider
    Source Movement Scale:Vector3f=&VectorFromFloat(Value=&FloatFromCurve(FloatCurve=DataInterface() {propertyValues {Curve {keys [<at 0 1 RCIM_Cubic>, {arriveTangent=-3.7372, interpMode=RCIM_Cubic, leaveTangent=-3.7372, tangentMode=RCTM_User, time=1, value=0}]}}}))
   )
   &AccelerationForce(Acceleration=&Add_Vector(A=<0 0 500>, B=&TransformVector(Vector=<0 1000 0>, Destination Space=::World)))
   &Drag(Drag=6, Use Linear Drag=true, Use Rotational Drag=false)
   &SolveForcesAndVelocity(Clamp Velocity=false, Limit Acceleration=false)
   &SubUVAnimation()
   &DynamicMaterialParameters(
    Index 0 Param 1=&FloatFromCurve(FloatCurve=DataInterface() {propertyValues {Curve {keys [<at 0 1>,<at 1 -1>]}}})
    Index 0 Param 2=1
    Index 0 Param 3=1
    Index 0 Param 4=1
    Module.Param0WriteEnabled=true
    Module.Param1WriteEnabled=true
    Module.Param2WriteEnabled=true
    Module.Param3WriteEnabled=true
   )
   &ScaleSpriteSize(Uniform Curve Sprite Scale=DataInterface() {propertyValues {Curve {keys [{tangentMode=RCTM_User, time=0, value=0.25}, {tangentMode=RCTM_User, time=1, value=2}]}}})
  ]
  rendererClasses [&NiagaraSpriteRendererProperties]
  renderers [{rendererClass=&NiagaraSpriteRendererProperties}]
 }
 Sprite_Flames {
  emitterUpdateScript [
   &EmitterState()
   &SpawnRate(SpawnRate=7, Use Spawn Probability=false)
   &SpawnPerUnit(
    Module.Use Max Movement Threshold=false
    Module.Use Movement Tolerance=true
    Movement Tolerance=0.1
    Spawn Group=1
    Spawn Probability=1
    Spawn Spacing=50
    Use Spawn Probability=true
   )
   &SpawnRate(SpawnGroup=2, SpawnRate=7, Use Spawn Probability=false) {module=SpawnRate001}
   &SpawnRate(SpawnGroup=3, SpawnRate=7, Use Spawn Probability=false) {module=SpawnRate002}
  ]
  particleSpawnScript [
   &InitializeParticle(
    Color=<rgba 1.78 5 0.026 0.5>
    Lifetime=&MaskFloatBySpawnGroup(Masked Float=0.25, Passthrough Float=&RandomRangeFloat(Minimum=0.25, Maximum=1), Spawn Group Mask=1)
    Sprite Rotation Mode=::NewEnumerator2
    Sprite Size Mode=::NewEnumerator1
    Uniform Sprite Size Max=&SelectFloatFromArray(
     Direct Array Index=&SpawnGroup()
     Float Selection Array=DataInterface() {propertyValues {FloatData [65,65,32.5,32.5]}}
    )
    Uniform Sprite Size Min=&SelectFloatFromArray(
     Direct Array Index=&SpawnGroup()
     Float Selection Array=DataInterface() {propertyValues {FloatData [15,15,7.5,7.5]}}
    )
    UsePositionOffset=false
    Write Material Random=true
   )
   &CameraOffset(Camera Offset Amount=10)
   &AddVelocity(Velocity=&MaskVectorBySpawnGroup(
     Masked Vector=&RandomRangeVector(Minimum=<0 50 0>, Maximum=<0 100 0>)
     Passthrough Vector=&RandomRangeVector(Minimum=<0 25 0>, Maximum=<0 50 0>)
     Spawn Group Mask=1
    ))
   /SetVariables(Particles.VisibilityTag=&SelectIntFromArray(Direct Array Index=&SpawnGroup(), Int Selection Array=DataInterface() {propertyValues {IntData [0,0,1,1]}})) {isSetParameters=true}
  ]
  particleUpdateScript [
   &ParticleState()
   &SkeletalMeshLocation(
    Bone / Socket Index=7
    Bone Sampling Mode=::NewEnumerator4
    MaskSpawnGroup=false
    Position Sampling=::NewEnumerator1
    Sampled Position Offset=<10 0 0>
    Skeletal Mesh=$User.Glider
    Source Movement Scale:Vector3f=&MaskVectorBySpawnGroup(Masked Vector=<1 1 1>, Passthrough Vector=<0 0 0>)
   )
   &ShapeLocation(
    Hemisphere Distribution=<0.5 1>
    MaskSpawnGroup=true
    Override Seed=false
    Rotation Axis=<0 -1 0>
    Rotation Mode=::NewEnumerator1
    Shape Origin=$Particles.SkeletalMeshLocation.SampledPosition
    Sphere Radius=10
    UseOwnerScale=true
   )
   &SkeletalMeshLocation(
    Bone / Socket Index=&MaskIntBySpawnGroup(Masked Int=1, Passthrough Int=2, Spawn Group Mask=2)
    MaskSpawnGroup=false
    Mesh Sampling Type=::NewEnumerator5
    Orientation Sampling=::NewEnumerator1
    Position Sampling=::NewEnumerator2
    Sampled Position Offset=<-2 0 0>
    Skeletal Mesh=$User.Glider
    Socket Sampling Mode=::NewEnumerator5
   ) {module=SkeletalMeshLocation001}
   &ShapeLocation(
    Apply To Particle Position=::NewEnumerator1
    Hemisphere Distribution=<0.5 1>
    MaskSpawnGroup=false
    Override Seed=false
    Rotation Axis=<0 -1 0>
    Rotation Mode=::NewEnumerator1
    Shape Origin=$Particles.SkeletalMeshLocation001.SampledPosition
    Sphere Radius=5
    UseOwnerScale=true
   ) {module=ShapeLocation001}
   /SetVariables(Particles.Exhaust Postion:NiagaraPosition=$Particles.ShapeLocation001.ShapePosition) {isSetParameters=true}
   &SubUVAnimation()
   &ScaleColor(
    Scale Alpha=&FloatFromCurve(FloatCurve=DataInterface() {propertyValues {Curve {keys [
         {arriveTangent=18.5059, interpMode=RCIM_Cubic, leaveTangent=18.5059, tangentMode=RCTM_User, time=0, value=0}
         {arriveTangentWeight=0.181, interpMode=RCIM_Cubic, leaveTangentWeight=0.7242, tangentMode=RCTM_User, time=0.2, value=1}
         {arriveTangent=-0.0405, interpMode=RCIM_Cubic, leaveTangent=-0.0405, tangentMode=RCTM_User, time=1, value=0}
        ]}}})
    Scale RGB=&MakeVectorFromLinearColorRGB(Color=&ColorFromCurve(ColorCurve=DataInterface() {propertyValues {
        AlphaCurve {keys []}
        BlueCurve {keys [<at 0 0 RCIM_Cubic>,<at 1 0 RCIM_Cubic>]}
        GreenCurve {keys [<at 0 1 RCIM_Cubic>,<at 1 1 RCIM_Cubic>]}
        RedCurve {keys [<at 0 0.3775 RCIM_Cubic>,<at 1 0.6175 RCIM_Cubic>]}
       }}))
    ScaleA=true
    ScaleRGB=true
   )
   &ScaleSpriteSize(Uniform Curve Sprite Scale=DataInterface() {propertyValues {Curve {keys [
        {arriveTangent=12.925, interpMode=RCIM_Cubic, leaveTangent=12.925, tangentMode=RCTM_User, time=0, value=0}
        {arriveTangentWeight=0.0886, interpMode=RCIM_Cubic, leaveTangentWeight=0.3543, tangentMode=RCTM_User, time=0.2, value=1}
        {interpMode=RCIM_Cubic, leaveTangent=-0.0405, time=1, value=0.875}
       ]}}})
   &InheritVelocity(
    Inherited Velocity Amount Scale=&MaskVectorBySpawnGroup(Masked Vector=<0 0 0>, Passthrough Vector=<-0.1 0 -0.1>)
    Speed Limit=true
    UseSpeedThreshold=false
   )
   &AccelerationForce(Acceleration=&MaskVectorBySpawnGroup(Masked Vector=<0 0 30>, Passthrough Vector=<0 0 0>))
   &Drag(Use Linear Drag=true, Use Rotational Drag=false)
   &SolveForcesAndVelocity(Clamp Velocity=false, Limit Acceleration=false)
  ]
  rendererClasses [&NiagaraSpriteRendererProperties]
  renderers [
   {rendererClass=&NiagaraSpriteRendererProperties}
   {rendererClass=&NiagaraSpriteRendererProperties, rendererIndex=1}
  ]
 }
 Train_Wheel_Chimney_Flame_Meshes {
  emitterUpdateScript [
   &EmitterState(Life Cycle Mode=::NewEnumerator1, Loop Behavior=::NewEnumerator1, Loop Duration Mode=::NewEnumerator1, UseLoopDelay=false)
   &SpawnBurst_Instantaneous(Use Spawn Probability=false, UseLoopCountLimit=false)
   &SpawnBurst_Instantaneous(Spawn Group=1, Use Spawn Probability=false, UseLoopCountLimit=false) {module=SpawnBurst_Instantaneous001}
   &SpawnBurst_Instantaneous(Spawn Group=2, Use Spawn Probability=false, UseLoopCountLimit=false) {module=SpawnBurst_Instantaneous002}
   &SpawnBurst_Instantaneous(Spawn Group=3, Use Spawn Probability=false, UseLoopCountLimit=false) {module=SpawnBurst_Instantaneous003}
   &SpawnBurst_Instantaneous(Spawn Group=4, Use Spawn Probability=false, UseLoopCountLimit=false) {module=SpawnBurst_Instantaneous004}
   &SpawnBurst_Instantaneous(Spawn Group=5, Use Spawn Probability=false, UseLoopCountLimit=false) {module=SpawnBurst_Instantaneous005}
   &SpawnBurst_Instantaneous(Spawn Group=6, Use Spawn Probability=false, UseLoopCountLimit=false) {module=SpawnBurst_Instantaneous006}
   &SpawnBurst_Instantaneous(Spawn Group=7, Use Spawn Probability=false, UseLoopCountLimit=false) {module=SpawnBurst_Instantaneous007}
   &SpawnBurst_Instantaneous(Spawn Group=8, Use Spawn Probability=false, UseLoopCountLimit=false) {module=SpawnBurst_Instantaneous008}
   &SpawnBurst_Instantaneous(Spawn Group=9, Use Spawn Probability=false, UseLoopCountLimit=false) {module=SpawnBurst_Instantaneous009}
   /SetVariables(
    Emitter.Wheel Opacity:NiagaraFloat=&Lerp_Float(A=0, B=1.5, Alpha=$System.Timeline.NormalizedTime)
    Emitter.Wheel Speed 2:NiagaraFloat=&Lerp_Float(
     A=&MakeCustomFloatFromBool(Bool=&SetBoolByFloatComparison(A=$User.ForwardDot, B=0), True Float=-10, False Float=10)
     B=&MakeCustomFloatFromBool(Bool=&SetBoolByFloatComparison(A=$User.ForwardDot, B=0), True Float=-30, False Float=30)
     Alpha=$System.Remap Forward Dot
    )
    Emitter.WheelSpeed:NiagaraFloat=&Lerp_Float(
     A=&MakeCustomFloatFromBool(Bool=&SetBoolByFloatComparison(A=$User.ForwardDot, B=0), True Float=10, False Float=-10)
     B=&MakeCustomFloatFromBool(Bool=&SetBoolByFloatComparison(A=$User.ForwardDot, B=0), True Float=30, False Float=-30)
     Alpha=$System.Remap Forward Dot
    )
   ) {isSetParameters=true}
   /SetVariables(
    Emitter.Chimney Flames Movement:Vector3f=$System.DetermineMovement.Movement
    Emitter.Chimney Flames Opacity Mask:NiagaraFloat=&Lerp_Float(A=0, B=50, Alpha=$System.Timeline.NormalizedTime)
   ) {isSetParameters=true}
   /SetVariables(
    Emitter.Flames Movement:Vector3f=$System.DetermineMovement.Movement
    Emitter.Flames Opacity Mask:NiagaraFloat=&Lerp_Float(A=0, B=100, Alpha=$System.Timeline.NormalizedTime)
   ) {isSetParameters=true}
   /SetVariables(
    Emitter.Exhaust Movement:Vector3f=$System.DetermineMovement001.Movement
    Emitter.Exhaust Opacity Mask:NiagaraFloat=&Lerp_Float(A=0, B=50, Alpha=$System.Timeline.NormalizedTime)
   ) {isSetParameters=true}
  ]
  particleSpawnScript [
   /SetVariables(Particles.VisibilityTag=&SelectIntFromArray(
     Direct Array Index=&SpawnGroup()
     Int Selection Array=DataInterface() {propertyValues {IntData [0,0,0,1,1,1,2,3,4]}}
    )) {isSetParameters=true}
   &InitializeParticle(
    Color Mode=::NewEnumerator0
    Lifetime=0
    Mesh Scale=&SelectVectorFromArray(
     Direct Array Index=&SpawnGroup()
     Vector Selection Array=DataInterface() {propertyValues {FloatData [<0.7 0.7 1>, <1 1 1>, <1 1 1>, <0.7 0.7 1>, <1 1 1>, <1 1 1>, <0.55 0.55 0.3>, <1 1 1>, <0.25 0.25 0.25>]}}
    )
    Mesh Scale Mode=::NewEnumerator4
    UsePositionOffset=false
    Write Material Random=true
   )
   &InitialMeshOrientation(
    Mesh Orientation Mode=::NewEnumerator1
    Rotation=&SelectVectorFromArray(
     Direct Array Index=&SpawnGroup()
     Vector Selection Array=DataInterface() {propertyValues {FloatData [<0 0 0>, <0 0 0>, <0 0 0>, <0 0 0>, <0 0 0>, <0 0 0>, <0 0 0>, <0 0 0>, <-0.25 0 0>]}}
    )
    Use Rotation Vector=true
   )
  ]
  particleUpdateScript [
   &ParticleState(Kill Particles When Lifetime Has Elapsed=false, Let Infinitely Lived Particles Die When Emitter Deactivates=true)
   &SkeletalMeshLocation(
    Bone / Socket Index=&SelectIntFromArray(
     Direct Array Index=&SpawnGroup()
     Int Selection Array=DataInterface() {propertyValues {IntData [1,2,3,4,5,6,7,8]}}
    )
    Bone Sampling Mode=::NewEnumerator4
    MaskSpawnGroup=false
    Orientation Sampling=::NewEnumerator1
    Skeletal Mesh=$User.Glider
   )
   &SkeletalMeshLocation(
    Bone / Socket Index=&SelectIntFromArray(
     Direct Array Index=&SpawnGroup()
     Int Selection Array=DataInterface() {propertyValues {IntData [0,0,0,0,0,0,0,0,1,2]}}
    )
    MaskSpawnGroup=false
    Mesh Sampling Type=::NewEnumerator5
    Orientation Sampling=::NewEnumerator1
    Position Sampling=::NewEnumerator2
    Sampled Position Offset=&MaskVectorBySpawnGroup(Masked Vector=<8 0 -2>, Passthrough Vector=<8 0 1>, Spawn Group Mask=8)
    Skeletal Mesh=$User.Glider
    Socket Sampling Mode=::NewEnumerator5
   ) {module=SkeletalMeshLocation001}
   /SetVariables(Particles.Exhaust Meshes Location:NiagaraPosition=$Particles.SkeletalMeshLocation001.SampledPosition) {isSetParameters=true}
  ]
  rendererClasses [&NiagaraMeshRendererProperties]
  renderers [
   {rendererClass=&NiagaraMeshRendererProperties, rendererIndex=0}
   {rendererClass=&NiagaraMeshRendererProperties, rendererIndex=1}
   {rendererClass=&NiagaraMeshRendererProperties, rendererIndex=2}
   {rendererClass=&NiagaraMeshRendererProperties, rendererIndex=3}
   {rendererClass=&NiagaraMeshRendererProperties, rendererIndex=4}
  ]
 }
 Whistle {
  emitterUpdateScript [
   &EmitterState(Life Cycle Mode=::NewEnumerator1, Loop Behavior=::NewEnumerator1, Loop Delay=0.2, Loop Duration=1.85, UseLoopDelay=true)
   &SpawnRate(SpawnRate=40, Use Spawn Probability=false)
  ]
  particleSpawnScript [
   &InitializeParticle(
    Color=&MakeLinearColorFromVectorAndFloat("Vector (RGB)"=<1 1 1>, "Float (Alpha)"=&RandomRangeFloat(Minimum=0.25, Maximum=0.5))
    Lifetime Max=0.4
    Lifetime Min=0.15
    Lifetime Mode=::NewEnumerator1
    Sprite Size Max=<40 100>
    Sprite Size Min=<20 50>
    Sprite Size Mode=::NewEnumerator2
    UsePositionOffset=false
    Write Material Random=true
   )
   &AddVelocity(Velocity=&RandomRangeVector(Minimum=<0 250 0>, Maximum=<0 500 0>))
  ]
  particleUpdateScript [
   &ParticleState()
   &SkeletalMeshLocation(
    Bone / Socket Index=9
    Bone Sampling Mode=::NewEnumerator4
    MaskSpawnGroup=false
    Position Sampling=::NewEnumerator1
    Sampled Position Offset=<18 5 25>
    Skeletal Mesh=$User.Glider
   )
   &SubUVAnimation()
   &DynamicMaterialParameters(
    Index 0 Param 1=&FloatFromCurve(FloatCurve=DataInterface() {propertyValues {Curve {keys [<at 0 1>,<at 1 -1>]}}})
    Index 0 Param 2=1
    Index 0 Param 3=1
    Index 0 Param 4=1
    Module.Param0WriteEnabled=true
    Module.Param1WriteEnabled=true
    Module.Param2WriteEnabled=true
    Module.Param3WriteEnabled=true
   )
   &ScaleSpriteSize(
    Non-Uniform Curve Sprite Scale=DataInterface() {propertyValues {
      XCurve {keys [<at 0 0 RCIM_Cubic>,<at 1 2 RCIM_Cubic>]}
      YCurve {keys [{tangentMode=RCTM_User, time=0, value=1}, {tangentMode=RCTM_User, time=1, value=1}]}
     }}
    Scale Sprite Size Mode=::NewEnumerator3
   )
   &AccelerationForce(Acceleration=<-500 0 2500>)
   &Drag(Drag=6, Use Linear Drag=true, Use Rotational Drag=false)
   &SolveForcesAndVelocity(Clamp Velocity=false, Limit Acceleration=false)
  ]
  rendererClasses [&NiagaraSpriteRendererProperties]
  renderers [{rendererClass=&NiagaraSpriteRendererProperties}]
 }
}
summary {
 systemName=NS_Glider_EvilCowboy_Main
 emitters [
  {
   emitterName=Lantern
   rendererClasses [
    &NiagaraSpriteRendererProperties
    &NiagaraMeshRendererProperties
    &NiagaraLightRendererProperties
   ]
  }
  {emitterName=Coal, rendererClasses [&NiagaraSpriteRendererProperties, &NiagaraMeshRendererProperties]}
  {emitterName=Coal_Light, rendererClasses [&NiagaraLightRendererProperties]}
  {emitterName=Train_Wheel_Chimney_Flame_Meshes, rendererClasses [&NiagaraMeshRendererProperties]}
  {emitterName=Sprite_Flames, rendererClasses [&NiagaraSpriteRendererProperties]}
  {emitterName=Glow_and_Lights, rendererClasses [&NiagaraMeshRendererProperties, &NiagaraLightRendererProperties]}
  {emitterName=Smoke, rendererClasses [&NiagaraSpriteRendererProperties]}
  {emitterName=Burst_Sparks, simTarget=GPUComputeSim, rendererClasses [&NiagaraSpriteRendererProperties]}
  {
   emitterName=Burst_Glow_and_Mesh
   rendererClasses [&NiagaraSpriteRendererProperties, &NiagaraMeshRendererProperties]
  }
  {emitterName=Whistle, rendererClasses [&NiagaraSpriteRendererProperties]}
  {bEnabled=false, emitterName=Burst_Sparks_Unopt, simTarget=GPUComputeSim, rendererClasses [&NiagaraSpriteRendererProperties]}
 ]
 userVariables [
  User.bisFullyDeployed {type=NiagaraBool, defaultValue=true}
  User.ForwardDot {type=NiagaraFloat, defaultValue=1, description=""}
  User.Glider {
   type=NiagaraDataInterfaceSkeletalMesh
   defaultValue {
    struct=NiagaraExt_VariableValue_DataInterface
    value {dataInterface=&self:NiagaraDataInterfaceSkeletalMesh_0}
   }
  }
  User.RightDot {type=NiagaraFloat, defaultValue=0, description=""}
 ]
}
