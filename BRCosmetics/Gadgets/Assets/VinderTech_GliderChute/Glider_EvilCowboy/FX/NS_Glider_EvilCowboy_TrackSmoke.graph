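# Niagara smoke and rune effect for the Evil Cowboy glider cosmetic, tracking six bones or sockets on the glider's skeletal mesh.
# Each spawn group follows its own socket, with purple smoke loops spinning in opposite directions and looping bursts of short-lived colored rune sprites around it.
# Two further smoke emitters gate spawn rate on User.ForwardDot; the speed-stretched one adds velocity when the comparison passes or in front-end preview.
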NiagaraSystem /BRCosmetics/Gadgets/Assets/VinderTech_GliderChute/Glider_EvilCowboy/FX/NS_Glider_EvilCowboy_TrackSmoke
dependencies {
 usedDataInterfaces [&NiagaraDataInterfaceArrayFloat, &NiagaraDataInterfaceCurve, &NiagaraDataInterfaceVector2DCurve]
 usedDynamicInputs []
 usedModules [&SystemState]
}
emitters {
 Runes {
  emitterUpdateScript [
   &EmitterState(Delay First Loop Only=true, Life Cycle Mode=::NewEnumerator1, Loop Delay=1, Loop Duration=0.4, UseLoopDelay=true)
   &SpawnBurst_Instantaneous(
    Spawn Count=&RandomRangeInt(Minimum=8, Maximum=12)
    Use Spawn Probability=false
    UseLoopCountLimit=false
   )
   &SpawnBurst_Instantaneous(
    Spawn Count=&RandomRangeInt(Minimum=8, Maximum=12)
    Spawn Group=1
    Use Spawn Probability=false
    UseLoopCountLimit=false
   ) {module=SpawnBurst_Instantaneous001}
   &SpawnBurst_Instantaneous(
    Spawn Count=&RandomRangeInt(Minimum=8, Maximum=12)
    Spawn Group=2
    Use Spawn Probability=false
    UseLoopCountLimit=false
   ) {module=SpawnBurst_Instantaneous002}
   &SpawnBurst_Instantaneous(
    Spawn Count=&RandomRangeInt(Minimum=8, Maximum=12)
    Spawn Group=3
    Use Spawn Probability=false
    UseLoopCountLimit=false
   ) {module=SpawnBurst_Instantaneous003}
   &SpawnBurst_Instantaneous(
    Spawn Count=&RandomRangeInt(Minimum=8, Maximum=12)
    Spawn Group=4
    Use Spawn Probability=false
    UseLoopCountLimit=false
   ) {module=SpawnBurst_Instantaneous004}
   &SpawnBurst_Instantaneous(
    Spawn Count=&RandomRangeInt(Minimum=8, Maximum=12)
    Spawn Group=5
    Use Spawn Probability=false
    UseLoopCountLimit=false
   ) {module=SpawnBurst_Instantaneous005}
  ]
  particleSpawnScript [
   &InitializeParticle(
    Color Maximum=<rgba 1 1.2 8 1>
    Color Minimum=<rgba 0 1 4 0.6>
    Color Mode=::NewEnumerator2
    Lifetime Max=0.8
    Lifetime Min=0.12
    Lifetime Mode=::NewEnumerator1
    Sprite Size Mode=::NewEnumerator1
    Uniform Sprite Size Max=15
    UsePositionOffset=false
    Write Material Random=true
   )
   &SubUVAnimation(End Frame=3, SubUV Animation Mode:ENiagara_SubUVLookupMode=::NewEnumerator2)
   &CameraOffset(Camera Offset Amount=10)
   &ShapeLocation(
    Apply To Particle Position=::NewEnumerator1
    Cylinder Height=0
    Cylinder Height Midpoint=0
    Cylinder Radius=&SelectFloatFromArray(
     Direct Array Index=&SpawnGroup()
     Float Selection Array=DataInterface() {propertyValues {FloatData [15,30,30,15,30,30]}}
    )
    MaskSpawnGroup=false
    Offset=<0 0 15>
    Offset Mode=::NewEnumerator0
    Override Seed=false
    Shape Primitive=::NewEnumerator1
    Surface Only=true
    UseOwnerScale=true
   )
   &SkeletalMeshLocation(
    Bone / Socket Index=&SpawnGroup()
    Bone Sampling Mode=::NewEnumerator4
    MaskSpawnGroup=false
    Orientation Sampling=::NewEnumerator1
    Position Sampling=::NewEnumerator2
    Sampled Position Offset Coordinate Space:ENiagaraSkelMeshTransforms=::NewEnumerator0
    Skeletal Mesh=$User.Glider
   ) {module=SkeletalMeshLocation001}
   &SpriteFacingAndAlignment(Sprite Facing=<0 0 1>, WriteAlignment=false, WriteFacing=true)
  ]
  particleUpdateScript [
   &ParticleState()
   &SkeletalMeshLocation(
    Bone / Socket Index=&SpawnGroup()
    Bone Sampling Mode=::NewEnumerator4
    MaskSpawnGroup=false
    Position Sampling=::NewEnumerator1
    Sampled Position Offset=/Niagara/DynamicInputs/Transforms/TransformPosition(Position=$Particles.ShapeLocation.ShapePosition, Source Space=::World, Destination Space=::Local)
    Sampled Position Offset Coordinate Space:ENiagaraSkelMeshTransforms=::NewEnumerator2
    Skeletal Mesh=$User.Glider
   )
   &ScaleColor(
    Scale Alpha=&FloatFromCurve(FloatCurve=DataInterface() {propertyValues {Curve {keys [{arriveTangent=-2.1944, interpMode=RCIM_Cubic, leaveTangent=-2.1944, tangentMode=RCTM_User, time=0, value=1}, <at 1 0 RCIM_Cubic>]}}})
    Scale RGB=&MakeVectorFromLinearColorRGB(Color=&RandomRangeLinearColor(Minimum=<rgba 1 4 0.2 1>, Maximum=<rgba 1 0.2 4 1>, Evaluation Type=::NewEnumerator1))
    ScaleA=true
    ScaleRGB=true
   )
  ]
  rendererClasses [&NiagaraSpriteRendererProperties]
  renderers [{rendererClass=&NiagaraSpriteRendererProperties}]
 }
 Smoke_Loop {
  emitterUpdateScript [
   &EmitterState()
   &SpawnRate(SpawnRate=&RandomRangeFloat(Minimum=6, Maximum=10), Use Spawn Probability=false)
   &SpawnRate(SpawnGroup=1, SpawnRate=&RandomRangeFloat(Minimum=6, Maximum=10), Use Spawn Probability=false) {module=SpawnRate001}
   &SpawnRate(SpawnGroup=2, SpawnRate=&RandomRangeFloat(Minimum=6, Maximum=10), Use Spawn Probability=false) {module=SpawnRate002}
   &SpawnRate(SpawnGroup=3, SpawnRate=&RandomRangeFloat(Minimum=6, Maximum=10), Use Spawn Probability=false) {module=SpawnRate003}
   &SpawnRate(SpawnGroup=4, SpawnRate=&RandomRangeFloat(Minimum=6, Maximum=10), Use Spawn Probability=false) {module=SpawnRate004}
   &SpawnRate(SpawnGroup=5, SpawnRate=&RandomRangeFloat(Minimum=6, Maximum=10), Use Spawn Probability=false) {module=SpawnRate005}
  ]
  particleSpawnScript [
   &InitializeParticle(
    Color=&MakeLinearColorFromVectorAndFloat(
     "Vector (RGB)"=&MakeVectorFromLinearColorRGB(Color=<rgba 1.11 0.0607 2 1>)
     "Float (Alpha)"=&RandomRangeFloat(Minimum=1, Maximum=2)
    )
    Lifetime Max=2.2
    Lifetime Min=1.7
    Lifetime Mode=::NewEnumerator1
    Sprite Rotation Mode=::NewEnumerator2
    Sprite Size Mode=::NewEnumerator1
    Uniform Sprite Size Max=75
    Uniform Sprite Size Min=50
    UsePositionOffset=false
    Write Material Random=true
   )
   &DynamicMaterialParameters(
    Index 0 Param 1=&RandomRangeFloat(Minimum=0, Maximum=1)
    Index 0 Param 2=&RandomRangeFloat(Minimum=0, Maximum=1)
    Module.Param0WriteEnabled=true
    Module.Param1WriteEnabled=true
    Module.Param2WriteEnabled=true
    Module.Param3WriteEnabled=true
   )
   &CameraOffset(Camera Offset Amount=&RandomRangeFloat(Minimum=15, Maximum=30))
   &ShapeLocation(
    Apply To Particle Position=::NewEnumerator1
    Cylinder Height=1
    Cylinder Height Midpoint=0
    Cylinder Radius=&SelectFloatFromArray(
     Direct Array Index=&SpawnGroup()
     Float Selection Array=DataInterface() {propertyValues {FloatData [20,32,32,20,32,32]}}
    )
    MaskSpawnGroup=false
    Offset=<0 0 10>
    Offset Mode=::NewEnumerator0
    Override Seed=false
    Shape Primitive=::NewEnumerator1
    Surface Only=true
    Use Endcaps In Surface Only Mode=false
    UseOwnerScale=true
   )
   &SkeletalMeshLocation(
    Bone / Socket Index=6
    Bone Sampling Mode=::NewEnumerator4
    MaskSpawnGroup=false
    Orientation Sampling=::NewEnumerator1
    Position Sampling=::NewEnumerator2
    Sampled Position Offset Coordinate Space:ENiagaraSkelMeshTransforms=::NewEnumerator0
    Skeletal Mesh=$User.Glider
   ) {module=SkeletalMeshLocation001}
   &AddVelocity(
    Additional Quaternion Rotation=false
    Rotation Mode=::NewEnumerator3
    Rotation Quaternion:Quat4f=$Particles.SkeletalMeshLocation001.SampledRotation
    Velocity=<0 25 0>
   )
  ]
  particleUpdateScript [
   &ParticleState()
   &SubUVAnimation(
    End Frame=60
    Start Frame=14
    SubUV Animation Mode:ENiagara_SubUVLookupMode=::NewEnumerator1
    SubUV Curve=DataInterface() {propertyValues {Curve {keys [{arriveTangent=1.7836, interpMode=RCIM_Cubic, leaveTangent=1.7836, tangentMode=RCTM_User, time=0, value=0}, <at 1 1 RCIM_Cubic>]}}}
   )
   &SkeletalMeshLocation(
    Bone / Socket Index=&SpawnGroup()
    Bone Sampling Mode=::NewEnumerator4
    MaskSpawnGroup=false
    Position Sampling=::NewEnumerator1
    Sampled Position Offset=&TransformPosition(Position=$Particles.ShapeLocation.ShapePosition, Source Space=::World, Destination Space=::Local)
    Sampled Position Offset Coordinate Space:ENiagaraSkelMeshTransforms=::NewEnumerator2
    Skeletal Mesh=$User.Glider
   )
   &ScaleColor(
    Scale Alpha=&FloatFromCurve(FloatCurve=DataInterface() {propertyValues {Curve {keys [
         {interpMode=RCIM_Cubic, tangentMode=RCTM_User, time=0, value=0}
         {interpMode=RCIM_Cubic, tangentMode=RCTM_User, time=0.5, value=1}
         {tangentMode=RCTM_User, time=1, value=0}
        ]}}})
    Scale RGB=&VectorFromFloat(Value=&FloatFromCurve(FloatCurve=DataInterface() {propertyValues {Curve {keys [{arriveTangent=-2.003, interpMode=RCIM_Cubic, leaveTangent=-2.003, tangentMode=RCTM_User, time=0, value=1}, <at 1 0.0987 RCIM_Cubic>]}}}))
    ScaleA=true
    ScaleRGB=true
   )
   &SpriteRotationRate(Rotation Rate=&RandomRangeFloat(Minimum=-90, Maximum=-217))
   &ScaleSpriteSize(Uniform Curve Sprite Scale=DataInterface() {propertyValues {Curve {keys [
        {arriveTangent=1.4881, interpMode=RCIM_Cubic, leaveTangent=1.4881, leaveTangentWeight=0.3333, tangentMode=RCTM_User, time=0, value=0.2}
        {arriveTangentWeight=0.3333, interpMode=RCIM_Cubic, tangentMode=RCTM_User, time=1, value=1.5}
       ]}}})
   &SolveForcesAndVelocity(Clamp Velocity=false, Limit Acceleration=false)
  ]
  rendererClasses [&NiagaraSpriteRendererProperties]
  renderers [{rendererClass=&NiagaraSpriteRendererProperties}]
 }
 Smoke_Loop_Reverse {
  emitterUpdateScript [
   &EmitterState()
   &SpawnRate(SpawnRate=&RandomRangeFloat(Minimum=6, Maximum=10), Use Spawn Probability=false)
   &SpawnRate(SpawnGroup=1, SpawnRate=&RandomRangeFloat(Minimum=6, Maximum=10), Use Spawn Probability=false) {module=SpawnRate001}
   &SpawnRate(SpawnGroup=2, SpawnRate=&RandomRangeFloat(Minimum=6, Maximum=10), Use Spawn Probability=false) {module=SpawnRate002}
   &SpawnRate(SpawnGroup=3, SpawnRate=&RandomRangeFloat(Minimum=6, Maximum=10), Use Spawn Probability=false) {module=SpawnRate003}
   &SpawnRate(SpawnGroup=4, SpawnRate=&RandomRangeFloat(Minimum=6, Maximum=10), Use Spawn Probability=false) {module=SpawnRate004}
   &SpawnRate(SpawnGroup=5, SpawnRate=&RandomRangeFloat(Minimum=6, Maximum=10), Use Spawn Probability=false) {module=SpawnRate005}
  ]
  particleSpawnScript [
   &InitializeParticle(
    Color=&MakeLinearColorFromVectorAndFloat(
     "Vector (RGB)"=&MakeVectorFromLinearColorRGB(Color=<rgba 1.11 0.0607 2 1>)
     "Float (Alpha)"=&RandomRangeFloat(Minimum=1, Maximum=2)
    )
    Lifetime Max=2.2
    Lifetime Min=1.7
    Lifetime Mode=::NewEnumerator1
    Sprite Rotation Mode=::NewEnumerator2
    Sprite Size Mode=::NewEnumerator1
    Uniform Sprite Size Max=75
    Uniform Sprite Size Min=50
    UsePositionOffset=false
    Write Material Random=true
   )
   &DynamicMaterialParameters(
    Index 0 Param 1=&RandomRangeFloat(Minimum=0, Maximum=1)
    Index 0 Param 2=&RandomRangeFloat(Minimum=0, Maximum=1)
    Module.Param0WriteEnabled=true
    Module.Param1WriteEnabled=true
    Module.Param2WriteEnabled=true
    Module.Param3WriteEnabled=true
   )
   &CameraOffset(Camera Offset Amount=&RandomRangeFloat(Minimum=15, Maximum=30))
   &ShapeLocation(
    Apply To Particle Position=::NewEnumerator1
    Cylinder Height=1
    Cylinder Height Midpoint=0
    Cylinder Radius=&SelectFloatFromArray(
     Direct Array Index=&SpawnGroup()
     Float Selection Array=DataInterface() {propertyValues {FloatData [20,32,32,20,32,32]}}
    )
    MaskSpawnGroup=false
    Offset=<0 0 10>
    Offset Mode=::NewEnumerator0
    Override Seed=false
    Shape Primitive=::NewEnumerator1
    Surface Only=true
    Use Endcaps In Surface Only Mode=false
    UseOwnerScale=true
   )
   &SkeletalMeshLocation(
    Bone / Socket Index=6
    Bone Sampling Mode=::NewEnumerator4
    MaskSpawnGroup=false
    Orientation Sampling=::NewEnumerator1
    Position Sampling=::NewEnumerator2
    Sampled Position Offset Coordinate Space:ENiagaraSkelMeshTransforms=::NewEnumerator0
    Skeletal Mesh=$User.Glider
   ) {module=SkeletalMeshLocation001}
   &AddVelocity(
    Additional Quaternion Rotation=false
    Rotation Mode=::NewEnumerator3
    Rotation Quaternion:Quat4f=$Particles.SkeletalMeshLocation001.SampledRotation
    Velocity=<0 25 0>
   )
  ]
  particleUpdateScript [
   &ParticleState()
   &SubUVAnimation(
    Start Frame=22
    SubUV Animation Mode:ENiagara_SubUVLookupMode=::NewEnumerator1
    SubUV Curve=DataInterface() {propertyValues {Curve {keys [{arriveTangent=1.7836, interpMode=RCIM_Cubic, leaveTangent=1.7836, tangentMode=RCTM_User, time=0, value=0}, <at 1 1 RCIM_Cubic>]}}}
   )
   &SkeletalMeshLocation(
    Bone / Socket Index=&SpawnGroup()
    Bone Sampling Mode=::NewEnumerator4
    MaskSpawnGroup=false
    Position Sampling=::NewEnumerator1
    Sampled Position Offset=&TransformPosition(Position=$Particles.ShapeLocation.ShapePosition, Source Space=::World, Destination Space=::Local)
    Sampled Position Offset Coordinate Space:ENiagaraSkelMeshTransforms=::NewEnumerator2
    Skeletal Mesh=$User.Glider
   )
   &ScaleColor(
    Scale Alpha=&FloatFromCurve(FloatCurve=DataInterface() {propertyValues {Curve {keys [
         {interpMode=RCIM_Cubic, tangentMode=RCTM_User, time=0, value=0}
         {interpMode=RCIM_Cubic, tangentMode=RCTM_User, time=0.5, value=1}
         {tangentMode=RCTM_User, time=1, value=0}
        ]}}})
    Scale RGB=&VectorFromFloat(Value=&FloatFromCurve(FloatCurve=DataInterface() {propertyValues {Curve {keys [{arriveTangent=-2.003, interpMode=RCIM_Cubic, leaveTangent=-2.003, tangentMode=RCTM_User, time=0, value=1}, <at 1 0.0987 RCIM_Cubic>]}}}))
    ScaleA=true
    ScaleRGB=true
   )
   &SpriteRotationRate(Rotation Rate=&RandomRangeFloat(Minimum=72, Maximum=180))
   &ScaleSpriteSize(Uniform Curve Sprite Scale=DataInterface() {propertyValues {Curve {keys [
        {arriveTangent=1.4881, interpMode=RCIM_Cubic, leaveTangent=1.4881, leaveTangentWeight=0.3333, tangentMode=RCTM_User, time=0, value=0.2}
        {arriveTangentWeight=0.3333, interpMode=RCIM_Cubic, tangentMode=RCTM_User, time=1, value=1.5}
       ]}}})
   &SolveForcesAndVelocity(Clamp Velocity=false, Limit Acceleration=false)
  ]
  rendererClasses [&NiagaraSpriteRendererProperties]
  renderers [{rendererClass=&NiagaraSpriteRendererProperties}]
 }
 Smoke_Loop_Velocity {
  emitterUpdateScript [
   &EmitterState()
   &SpawnRate(
    SpawnRate=&MakeCustomFloatFromBool(
     Bool=&Boolean_OrOperation(
      Bool1=$User.bisFrontEndPreview
      Bool2=&SetBoolByFloatComparison(A=$User.ForwardDot, B=0.2, Comparison Type=::NewEnumerator2)
     )
     True Float=&RandomRangeFloat(Minimum=6, Maximum=10)
     False Float=0
    )
    Use Spawn Probability=false
   )
   &SpawnRate(
    SpawnGroup=1
    SpawnRate=&MakeCustomFloatFromBool(
     Bool=&Boolean_OrOperation(
      Bool1=$User.bisFrontEndPreview
      Bool2=&SetBoolByFloatComparison(A=$User.ForwardDot, B=0.2, Comparison Type=::NewEnumerator2)
     )
     True Float=&RandomRangeFloat(Minimum=6, Maximum=10)
     False Float=0
    )
    Use Spawn Probability=false
   ) {module=SpawnRate001}
   &SpawnRate(
    SpawnGroup=2
    SpawnRate=&MakeCustomFloatFromBool(
     Bool=&Boolean_OrOperation(
      Bool1=$User.bisFrontEndPreview
      Bool2=&SetBoolByFloatComparison(A=$User.ForwardDot, B=0.2, Comparison Type=::NewEnumerator2)
     )
     True Float=&RandomRangeFloat(Minimum=6, Maximum=10)
     False Float=0
    )
    Use Spawn Probability=false
   ) {module=SpawnRate002}
   &SpawnRate(
    SpawnGroup=3
    SpawnRate=&MakeCustomFloatFromBool(
     Bool=&Boolean_OrOperation(
      Bool1=$User.bisFrontEndPreview
      Bool2=&SetBoolByFloatComparison(A=$User.ForwardDot, B=0.2, Comparison Type=::NewEnumerator2)
     )
     True Float=&RandomRangeFloat(Minimum=6, Maximum=10)
     False Float=0
    )
    Use Spawn Probability=false
   ) {module=SpawnRate003}
   &SpawnRate(
    SpawnGroup=4
    SpawnRate=&MakeCustomFloatFromBool(
     Bool=&Boolean_OrOperation(
      Bool1=$User.bisFrontEndPreview
      Bool2=&SetBoolByFloatComparison(A=$User.ForwardDot, B=0.2, Comparison Type=::NewEnumerator2)
     )
     True Float=&RandomRangeFloat(Minimum=6, Maximum=10)
     False Float=0
    )
    Use Spawn Probability=false
   ) {module=SpawnRate004}
   &SpawnRate(
    SpawnGroup=5
    SpawnRate=&MakeCustomFloatFromBool(
     Bool=&Boolean_OrOperation(
      Bool1=$User.bisFrontEndPreview
      Bool2=&SetBoolByFloatComparison(A=$User.ForwardDot, B=0.2, Comparison Type=::NewEnumerator2)
     )
     True Float=&RandomRangeFloat(Minimum=6, Maximum=10)
     False Float=0
    )
    Use Spawn Probability=false
   ) {module=SpawnRate005}
  ]
  particleSpawnScript [
   &InitializeParticle(
    Color=&MakeLinearColorFromVectorAndFloat(
     "Vector (RGB)"=&MakeVectorFromLinearColorRGB(Color=<rgba 1.11 0.0607 2 1>)
     "Float (Alpha)"=&RandomRangeFloat(Minimum=1, Maximum=2)
    )
    Lifetime Max=0.6
    Lifetime Min=0.4
    Lifetime Mode=::NewEnumerator1
    Position Offset=<0 0 10>
    Sprite Size Max=<90 90>
    Sprite Size Min=<45 45>
    Sprite Size Mode=::NewEnumerator2
    Sprite UV Mode=::NewEnumerator1
    UsePositionOffset=true
    Write Material Random=true
   )
   &CameraOffset(Camera Offset Amount=&RandomRangeFloat(Minimum=15, Maximum=30))
   &ShapeLocation(
    Apply To Particle Position=::NewEnumerator1
    Cylinder Height=1
    Cylinder Height Midpoint=0
    Cylinder Radius=&SelectFloatFromArray(
     Direct Array Index=&SpawnGroup()
     Float Selection Array=DataInterface() {propertyValues {FloatData [20,32,32,20,32,32]}}
    )
    MaskSpawnGroup=false
    Offset=<0 0 10>
    Offset Mode=::NewEnumerator0
    Override Seed=false
    Shape Primitive=::NewEnumerator1
    Surface Only=true
    Use Endcaps In Surface Only Mode=false
    UseOwnerScale=true
   ) {enabled=false}
   &SkeletalMeshLocation(
    Bone / Socket Index=6
    Bone Sampling Mode=::NewEnumerator4
    MaskSpawnGroup=false
    Orientation Sampling=::NewEnumerator1
    Position Sampling=::NewEnumerator2
    Sampled Position Offset Coordinate Space:ENiagaraSkelMeshTransforms=::NewEnumerator0
    Skeletal Mesh=$User.Glider
   ) {module=SkeletalMeshLocation001}
   &AddVelocity(
    Additional Quaternion Rotation=false
    Rotation Mode=::NewEnumerator3
    Rotation Quaternion:Quat4f=$Particles.SkeletalMeshLocation001.SampledRotation
    Velocity=&Add_Vector(
     A=<0 25 0>
     B=&MakeCustomVectorFromBool(
      Bool=&Boolean_OrOperation(
       Bool1=$User.bisFrontEndPreview
       Bool2=&SetBoolByFloatComparison(A=$User.ForwardDot, B=0.2, Comparison Type=::NewEnumerator2)
      )
      True Vector=<0 250 0>
      False Vector=<0 0 0>
     )
    )
   )
  ]
  particleUpdateScript [
   &ParticleState()
   &DynamicMaterialParameters(
    Index 0 Param 1=&RandomRangeFloat(Minimum=0, Maximum=1)
    Index 0 Param 2=1
    Index 0 Param 3=&FloatFromCurve(FloatCurve=DataInterface() {propertyValues {Curve {keys [
         <at 0.5 1 RCIM_Cubic>
         {arriveTangent=-7.3677, interpMode=RCIM_Cubic, leaveTangent=-7.3677, tangentMode=RCTM_User, time=0.6598, value=0.5089}
         <at 0.8 0 RCIM_Cubic>
        ]}}})
    Module.Param0WriteEnabled=true
    Module.Param1WriteEnabled=true
    Module.Param2WriteEnabled=true
    Module.Param3WriteEnabled=true
   )
   &SolveForcesAndVelocity(Clamp Velocity=false, Limit Acceleration=false)
   &SubUVAnimation(
    End Frame=39
    Start Frame=9
    SubUV Animation Mode:ENiagara_SubUVLookupMode=::NewEnumerator1
    SubUV Curve=DataInterface() {propertyValues {Curve {keys [{arriveTangent=1.7836, interpMode=RCIM_Cubic, leaveTangent=1.7836, tangentMode=RCTM_User, time=0, value=0}, <at 1 1 RCIM_Cubic>]}}}
   )
   &SkeletalMeshLocation(
    Bone / Socket Index=&SpawnGroup()
    Bone Sampling Mode=::NewEnumerator4
    MaskSpawnGroup=false
    Position Sampling=::NewEnumerator1
    Sampled Position Offset Coordinate Space:ENiagaraSkelMeshTransforms=::NewEnumerator0
    Skeletal Mesh=$User.Glider
   )
   &ScaleColor(
    Scale Alpha=&FloatFromCurve(FloatCurve=DataInterface() {propertyValues {Curve {keys [
         {interpMode=RCIM_Cubic, tangentMode=RCTM_User, time=0, value=0}
         {interpMode=RCIM_Cubic, tangentMode=RCTM_User, time=0.5, value=1}
         {tangentMode=RCTM_User, time=1, value=0}
        ]}}})
    Scale RGB=&VectorFromFloat(Value=&FloatFromCurve(FloatCurve=DataInterface() {propertyValues {Curve {keys [{arriveTangent=-1.2967, interpMode=RCIM_Cubic, leaveTangent=-1.2967, tangentMode=RCTM_User, time=0, value=1}, <at 1 0.25 RCIM_Cubic>]}}}))
    ScaleA=true
    ScaleRGB=true
   )
   &ScaleSpriteSize(
    Non-Uniform Curve Sprite Scale=DataInterface() {propertyValues {
      XCurve {keys [
        {arriveTangent=2.9988, interpMode=RCIM_Cubic, leaveTangent=2.9988, tangentMode=RCTM_User, time=0, value=0.05}
        {arriveTangent=1.7776, interpMode=RCIM_Cubic, leaveTangent=1.7776, time=0.2344, value=0.7419}
        {arriveTangent=0.5984, interpMode=RCIM_Cubic, leaveTangent=0.5984, time=0.5916, value=1.1017}
        <at 1 1.2 RCIM_Cubic>
       ]}
      YCurve {keys [
        {arriveTangent=4.5379, interpMode=RCIM_Cubic, leaveTangent=4.5379, tangentMode=RCTM_User, time=0, value=0.1}
        {arriveTangent=2.17, interpMode=RCIM_Cubic, leaveTangent=2.17, time=0.2344, value=1.0157}
        {arriveTangent=0.6325, interpMode=RCIM_Cubic, leaveTangent=0.6325, time=0.5916, value=1.3838}
        <at 1 1.5 RCIM_Cubic>
       ]}
     }}
    Scale Sprite Size Mode=::NewEnumerator3
   )
   &ScaleSpriteSizeBySpeed(Max Scale Factor=<1 5>)
  ]
  rendererClasses [&NiagaraSpriteRendererProperties]
  renderers [{rendererClass=&NiagaraSpriteRendererProperties}]
 }
 Smoke_Loop_Velocity_Idle {
  emitterUpdateScript [
   &EmitterState()
   &SpawnRate(
    SpawnRate=&MakeCustomFloatFromBool(
     Bool=&SetBoolByFloatComparison(A=$User.ForwardDot, B=0.2, Comparison Type=::NewEnumerator2)
     True Float=&RandomRangeFloat(Minimum=6, Maximum=10)
     False Float=2
    )
    Use Spawn Probability=false
   )
   &SpawnRate(
    SpawnGroup=1
    SpawnRate=&MakeCustomFloatFromBool(
     Bool=&SetBoolByFloatComparison(A=$User.ForwardDot, B=0.2, Comparison Type=::NewEnumerator2)
     True Float=&RandomRangeFloat(Minimum=6, Maximum=10)
     False Float=2
    )
    Use Spawn Probability=false
   ) {module=SpawnRate001}
   &SpawnRate(
    SpawnGroup=2
    SpawnRate=&MakeCustomFloatFromBool(
     Bool=&SetBoolByFloatComparison(A=$User.ForwardDot, B=0.2, Comparison Type=::NewEnumerator2)
     True Float=&RandomRangeFloat(Minimum=6, Maximum=10)
     False Float=2
    )
    Use Spawn Probability=false
   ) {module=SpawnRate002}
   &SpawnRate(
    SpawnGroup=3
    SpawnRate=&MakeCustomFloatFromBool(
     Bool=&SetBoolByFloatComparison(A=$User.ForwardDot, B=0.2, Comparison Type=::NewEnumerator2)
     True Float=&RandomRangeFloat(Minimum=6, Maximum=10)
     False Float=2
    )
    Use Spawn Probability=false
   ) {module=SpawnRate003}
   &SpawnRate(
    SpawnGroup=4
    SpawnRate=&MakeCustomFloatFromBool(
     Bool=&SetBoolByFloatComparison(A=$User.ForwardDot, B=0.2, Comparison Type=::NewEnumerator2)
     True Float=&RandomRangeFloat(Minimum=6, Maximum=10)
     False Float=2
    )
    Use Spawn Probability=false
   ) {module=SpawnRate004}
   &SpawnRate(
    SpawnGroup=5
    SpawnRate=&MakeCustomFloatFromBool(
     Bool=&SetBoolByFloatComparison(A=$User.ForwardDot, B=0.2, Comparison Type=::NewEnumerator2)
     True Float=&RandomRangeFloat(Minimum=6, Maximum=10)
     False Float=2
    )
    Use Spawn Probability=false
   ) {module=SpawnRate005}
  ]
  particleSpawnScript [
   &InitializeParticle(
    Color=&MakeLinearColorFromVectorAndFloat(
     "Vector (RGB)"=&MakeVectorFromLinearColorRGB(Color=<rgba 1.11 0.0607 2 1>)
     "Float (Alpha)"=&RandomRangeFloat(Minimum=1, Maximum=2)
    )
    Lifetime Max=1
    Lifetime Min=0.75
    Lifetime Mode=::NewEnumerator1
    Sprite Rotation Angle=&RandomRangeFloat(Minimum=0.45, Maximum=0.55)
    Sprite Rotation Mode=::NewEnumerator1
    Sprite Size Max=<90 90>
    Sprite Size Min=<45 45>
    Sprite Size Mode=::NewEnumerator2
    UsePositionOffset=false
    Write Material Random=true
   )
   &CameraOffset(Camera Offset Amount=&RandomRangeFloat(Minimum=15, Maximum=30))
   &ShapeLocation(
    Apply To Particle Position=::NewEnumerator1
    Cylinder Height=1
    Cylinder Radius=&SelectFloatFromArray(
     Direct Array Index=&SpawnGroup()
     Float Selection Array=DataInterface() {propertyValues {FloatData [20,37,37,20,37,37]}}
    )
    MaskSpawnGroup=false
    Override Seed=false
    Shape Primitive=::NewEnumerator1
    Surface Only=true
    Use Endcaps In Surface Only Mode=false
    UseOwnerScale=true
   ) {enabled=false}
   &SkeletalMeshLocation(
    Bone / Socket Index=6
    Bone Sampling Mode=::NewEnumerator4
    MaskSpawnGroup=false
    Orientation Sampling=::NewEnumerator1
    Position Sampling=::NewEnumerator2
    Sampled Position Offset Coordinate Space:ENiagaraSkelMeshTransforms=::NewEnumerator0
    Skeletal Mesh=$User.Glider
   ) {module=SkeletalMeshLocation001}
   &AddVelocity(
    Additional Quaternion Rotation=false
    Rotation Mode=::NewEnumerator3
    Rotation Quaternion:Quat4f=$Particles.SkeletalMeshLocation001.SampledRotation
    Velocity=<0 25 0>
   )
  ]
  particleUpdateScript [
   &ParticleState()
   &DynamicMaterialParameters(
    Index 0 Param 1=&RandomRangeFloat(Minimum=0, Maximum=1)
    Index 0 Param 2=1
    Index 0 Param 3=&FloatFromCurve(FloatCurve=DataInterface() {propertyValues {Curve {keys [
         <at 0.5 1 RCIM_Cubic>
         {arriveTangent=-7.3677, interpMode=RCIM_Cubic, leaveTangent=-7.3677, tangentMode=RCTM_User, time=0.6598, value=0.5089}
         <at 0.8 0 RCIM_Cubic>
        ]}}})
    Module.Param0WriteEnabled=true
    Module.Param1WriteEnabled=true
    Module.Param2WriteEnabled=true
    Module.Param3WriteEnabled=true
   )
   &SolveForcesAndVelocity(Clamp Velocity=false, Limit Acceleration=false)
   &SubUVAnimation(
    End Frame=58
    Start Frame=4
    SubUV Animation Mode:ENiagara_SubUVLookupMode=::NewEnumerator1
    SubUV Curve=DataInterface() {propertyValues {Curve {keys [
        {arriveTangent=4.2151, interpMode=RCIM_Cubic, leaveTangent=4.2151, tangentMode=RCTM_User, time=0, value=0}
        {arriveTangent=0.7423, interpMode=RCIM_Cubic, leaveTangent=0.7423, tangentMode=RCTM_User, time=0.2, value=0.5}
        <at 1 1 RCIM_Cubic>
       ]}}}
   )
   &SkeletalMeshLocation(
    Bone / Socket Index=&SpawnGroup()
    Bone Sampling Mode=::NewEnumerator4
    MaskSpawnGroup=false
    Position Sampling=::NewEnumerator1
    Sampled Position Offset Coordinate Space:ENiagaraSkelMeshTransforms=::NewEnumerator0
    Skeletal Mesh=$User.Glider
   )
   &ScaleColor(
    Scale Alpha=&FloatFromCurve(FloatCurve=DataInterface() {propertyValues {Curve {keys [
         {interpMode=RCIM_Cubic, tangentMode=RCTM_User, time=0, value=0}
         {interpMode=RCIM_Cubic, tangentMode=RCTM_User, time=0.5, value=1}
         {tangentMode=RCTM_User, time=1, value=0}
        ]}}})
    Scale RGB=&VectorFromFloat(Value=&FloatFromCurve(FloatCurve=DataInterface() {propertyValues {Curve {keys [{arriveTangent=-1.2967, interpMode=RCIM_Cubic, leaveTangent=-1.2967, tangentMode=RCTM_User, time=0, value=1}, <at 1 0.25 RCIM_Cubic>]}}}))
    ScaleA=true
    ScaleRGB=true
   )
   &ScaleSpriteSize(
    Non-Uniform Curve Sprite Scale=DataInterface() {propertyValues {
      XCurve {keys [
        {arriveTangent=2.9988, interpMode=RCIM_Cubic, leaveTangent=2.9988, tangentMode=RCTM_User, time=0, value=0.05}
        {arriveTangent=1.7776, interpMode=RCIM_Cubic, leaveTangent=1.7776, time=0.2344, value=0.7419}
        {arriveTangent=0.5984, interpMode=RCIM_Cubic, leaveTangent=0.5984, time=0.5916, value=1.1017}
        <at 1 1.2 RCIM_Cubic>
       ]}
      YCurve {keys [
        {arriveTangent=4.5379, interpMode=RCIM_Cubic, leaveTangent=4.5379, tangentMode=RCTM_User, time=0, value=0.1}
        {arriveTangent=2.17, interpMode=RCIM_Cubic, leaveTangent=2.17, time=0.2344, value=1.0157}
        {arriveTangent=0.6325, interpMode=RCIM_Cubic, leaveTangent=0.6325, time=0.5916, value=1.3838}
        <at 1 1.5 RCIM_Cubic>
       ]}
     }}
    Scale Sprite Size Mode=::NewEnumerator3
   )
  ]
  rendererClasses [&NiagaraSpriteRendererProperties]
  renderers [{rendererClass=&NiagaraSpriteRendererProperties}]
 }
}
summary {
 systemName=NS_Glider_EvilCowboy_TrackSmoke
 emitters [
  {emitterName=Smoke_Loop, rendererClasses [&NiagaraSpriteRendererProperties]}
  {emitterName=Smoke_Loop_Reverse, rendererClasses [&NiagaraSpriteRendererProperties]}
  {emitterName=Smoke_Loop_Velocity, rendererClasses [&NiagaraSpriteRendererProperties]}
  {emitterName=Smoke_Loop_Velocity_Idle, rendererClasses [&NiagaraSpriteRendererProperties]}
  {emitterName=Runes, rendererClasses [&NiagaraSpriteRendererProperties]}
 ]
 userVariables [
  User.bisFrontEndPreview {type=NiagaraBool}
  User.ForwardDot {type=NiagaraFloat, defaultValue=1, description=""}
  User.Glider {
   type=NiagaraDataInterfaceSkeletalMesh
   defaultValue {
    struct=NiagaraExt_VariableValue_DataInterface
    value {dataInterface=&self:NiagaraDataInterfaceSkeletalMesh_0}
   }
  }
 ]
}
